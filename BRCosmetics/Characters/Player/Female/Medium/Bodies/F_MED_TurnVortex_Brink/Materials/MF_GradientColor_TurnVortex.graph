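# Material function for the TurnVortex Brink character skin that outputs an animated gradient Color and a glow Mask.
# Projected UVs from a borrowed glider function are scaled and panned to sample the EmissiveGradient noise, which, with a sine-over-time offset, indexes a gradient texture.
# Inverted Fresnel and noise-driven glow intensity shade that color, which is desaturation-adjusted and blended with fixed teal.
# Mask multiplies the same glow by panned noise and sharpens it with a power.
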MaterialFunction /BRCosmetics/Characters/Player/Female/Medium/Bodies/F_MED_TurnVortex_Brink/Materials/MF_GradientColor_TurnVortex
graph {nodes [
  n1 = Constant3Vector() {constant=<rgba 0.02 0.409 0.5 1>}
  n2 = Time()
  n3 = Sine(None=Multiply(A=Multiply(A=@n2) {constB=0.1}, B=ScalarParameter() {group=TurnVortex, parameterName=SineSpeed, sortPriority=2}))
  n4 = NamedRerouteDeclaration(None=MaterialFunctionCall!(
    Root_Position=Constant()
    UseClamp?
    FlipUV
    UVSize=Constant() {r=450}
    CenterOffsetX=Constant()
    CenterOffsetY=Constant()
    CenterOffsetZ=Constant()
   ) {materialFunction=/BRCosmetics/Gadgets/Assets/VinderTech_GliderChute/Glider_WiryPerk/Materials/MF_ProjectUV_WiryPerk}.UV_BackProjection) {name=FrontProjectUV, nodeColor=<rgba 0.2422 0 1 1>}
  n5 = ComponentMask(None=ScalarParameter() {defaultValue=3.6634, group=TurnVortex, parameterName=EmissivePannerScale}) {
   desc="Kendall added this node on 5/27/2020.\r\nPreviously, this just piped the 3-vector value from\r\nthe texture sample straight into the UV's, which\r\ncaused the material to fail compile"
   g=true
   r=true
  }
  n6 = Panner(
   Coordinate=Add(A=Constant(), B=Multiply(A=@n4, B=@n5))
   Speed=AppendVector(
    A=ScalarParameter() {defaultValue=0.03, group=TurnVortex, parameterName=TurnVortexSpeed_X, sortPriority=2}
    B=ScalarParameter() {defaultValue=0.03, group=TurnVortex, parameterName=TurnVortexSpeed_Y, sortPriority=2}
   )
  )
  n7 = TextureObjectParameter() {
   group=TurnVortex
   parameterName=EmissiveGradient
   samplerType=SAMPLERTYPE_Grayscale
   sortPriority=1
   texture=/Game/Global/Textures/Default/Noise/LowResBlurredNoiseNormalized
  }
  n8 = TextureSample(UVs=@n6, Tex=@n7) {
   samplerSource=SSM_Wrap_WorldGroupSettings
   samplerType=SAMPLERTYPE_Masks
   texture=/Game/Weapons/WeaponSkins/Textures/T_Swirly_Noise_highContrast
  }
  n9 = Add(
   A=LinearInterpolate(B=ScalarParameter() {defaultValue=3.8444, group=TurnVortex, parameterName=ColorOffset}, Alpha=ConstantBiasScale(None=@n3)) {constB=0.5}
   B=@n8.RGB
  )
  n10 = Constant() {r=0.3008}
  n11 = TextureSampleParameter2D(UVs=ComponentMask(None=Multiply(A=@n9, B=@n10) {constB=0.5}) {g=true, r=true}) {
   group=TurnVortex
   parameterName=Gradient Color
   texture=/BRCosmetics/Accessories/FORT_Backpacks/Backpack_F_MED_LowerDrop/Material/T_Backpack_LowerDrop_Gradient
  }
  n12 = OneMinus(None=Fresnel(ExponentIn=Constant() {r=1}) {exponent=2})
  n13 = Saturate(None=LinearInterpolate(A=ScalarParameter() {defaultValue=0.4, group=TurnVortex, parameterName=Darkedge}, Alpha=Multiply(A=@n12) {constB=1.3}) {constA=0.3})
  n14 = LinearInterpolate(
   A=ScalarParameter() {defaultValue=1, group=TurnVortex, parameterName=GlowIntensityMax}
   B=ScalarParameter() {defaultValue=0.85, group=TurnVortex, parameterName=GlowIntensityMin}
   Alpha=@n8.R
  )
  n15 = Desaturation(
   None=Multiply(A=@n11.RGB, B=Multiply(A=@n13, B=@n14))
   Fraction=ScalarParameter() {defaultValue=-0.5, group=TurnVortex, parameterName=ColorDesaturation}
  )
  FunctionOutput(None=NamedRerouteDeclaration(None=LinearInterpolate(A=@n1, B=@n15)) {name=Gradient Color, nodeColor=<rgba 1 0 0.8906 1>}) {outputName=Color}
  n16 = TextureObjectParameter() {
   group=TurnVortex
   parameterName=EmissiveGradient
   samplerType=SAMPLERTYPE_Grayscale
   sortPriority=1
   texture=&LowResBlurredNoiseNormalized
  }
  n17 = TextureSample(UVs=@n6, Tex=@n16) {
   samplerType=SAMPLERTYPE_Masks
   texture=/BRCosmetics/Characters/Player/Female/Medium/Bodies/F_MED_TurnVortex_Brink/Textures/T_TurnVortex_Brink_Skin_FX
  }
  FunctionOutput(None=Power(Base=Multiply(A=@n14, B=@n17.R)) {constExponent=1.5}) {outputName=Mask}
 ]}
