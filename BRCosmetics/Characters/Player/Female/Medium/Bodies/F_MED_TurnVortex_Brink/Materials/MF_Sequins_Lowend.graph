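# Material function for the TurnVortex Brink character body that layers sequins onto incoming material attributes through TechArtMask, with static switches per channel.
# Rotatable tiled textures supply sequin offset, roughness and normals, colored by two falloff colors.
# Optional thin-film tint samples a spectrum by view angle, while emissive sparkle multiplies scrolling noise by a panning gradient.
# Quality switches swap emissive and roughness for constants on low and mobile tiers.
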MaterialFunction /BRCosmetics/Characters/Player/Female/Medium/Bodies/F_MED_TurnVortex_Brink/Materials/MF_Sequins_Lowend
graph {nodes [
  n1 = FunctionInput() {bUsePreviewValueAsDefault=true, inputName=inMat, inputType=FunctionInput_MaterialAttributes, sortPriority=-1}
  n2 = GetMaterialAttributes(None=@n1) {attributeGetTypes [
    #1
    #2
    #3
    #4
   ]}
  n3 = GetMaterialAttributes(None=NamedRerouteDeclaration(None=@n1) {name=Mat_IN, nodeColor=<rgba 1 0.0469 0 1>}) {attributeGetTypes [
    #5
    #4
    #1
    #6
    #2
    #3
   ]}
  n4 = FunctionInput(Preview=Constant() {r=1}) {inputName=TechArtMask, inputType=FunctionInput_Scalar, sortPriority=8}
  n5 = NamedRerouteUsage()
  n6 = Constant()
  n7 = NamedRerouteUsage()
  n8 = NamedRerouteUsage()
  n9 = NamedRerouteUsage()
  n10 = NamedRerouteUsage()
  n11 = SetMaterialAttributes!(
   MaterialAttributes=@n2.MaterialAttributes
   BaseColor=StaticSwitchParameter(True=LinearInterpolate(A=@n3.BaseColor, B=@n5, Alpha=@n4), False=@n6) {defaultValue=true, group=09_DisableTextureSamplers, parameterName=MFSequin_UseBaseColor}
   Metallic=StaticSwitchParameter(True=LinearInterpolate(A=@n3.Metallic, B=@n7, Alpha=@n4), False=@n6) {defaultValue=true, group=09_DisableTextureSamplers, parameterName=MFSequin_UseMetallic}
   Roughness=StaticSwitchParameter(True=LinearInterpolate(A=@n3.Roughness, B=@n8, Alpha=@n4), False=@n6) {defaultValue=true, group=09_DisableTextureSamplers, parameterName=MFSequin_UseRoughness}
   EmissiveColor=StaticSwitchParameter(True=LinearInterpolate(A=@n3.EmissiveColor, B=@n9, Alpha=@n4), False=@n6) {defaultValue=true, group=09_DisableTextureSamplers, parameterName=MFSequin_UseEmissiveColor}
   Normal=StaticSwitchParameter(True=LinearInterpolate(A=@n3.Normal, B=@n10, Alpha=@n4), False=@n6) {defaultValue=true, group=09_DisableTextureSamplers, parameterName=MFSequin_UseNormal}
   Specular
  ) {attributeSetTypes [
    #5
    #4
    #1
    #6
    #2
    #3
   ]}
  FunctionOutput(None=StaticSwitchParameter(True=@n11, False=NamedRerouteUsage()) {defaultValue=true, group=04_Tech Art, parameterName=UseSequins})
  n12 = Constant()
  n13 = NamedRerouteUsage()
  n14 = ScalarParameter() {group=Sequin, parameterName=SequinThinFilmStrength_Emissive}
  n15 = NamedRerouteUsage()
  n16 = Time()
  n17 = ScalarParameter() {defaultValue=0.01, group=Sequin, parameterName=SequinSparkleSpeed, sortPriority=12}
  n18 = TextureSample(UVs=Add(A=ComponentMask(None=@n15) {g=true, r=true}, B=Multiply(A=@n16, B=@n17) {constB=0.2})) {samplerSource=SSM_Wrap_WorldGroupSettings, samplerType=SAMPLERTYPE_Normal, texture=/Game/Global/Textures/Default/Noise/64x64Noise}
  n19 = Power(Base=@n18.R) {constExponent=4}
  n20 = TextureCoordinate()
  n21 = ScalarParameter() {defaultValue=2, group=Sequin, parameterName=SequinEmissiveTiling, sortPriority=14}
  n22 = VectorParameter() {group=Sequin, parameterName=SequinEmissivePanSpeed, sortPriority=14, defaultValue=<rgba 0.05 0.05 0 0>}
  n23 = TextureSampleParameter2D(UVs=Panner(Coordinate=Multiply(A=@n20, B=@n21), Speed=ComponentMask(None=@n22.RGB) {g=true, r=true}) {speedX=0.05, speedY=0.05}) {
   group=Sequin
   parameterName=SequinEmissiveGradient
   samplerSource=SSM_Wrap_WorldGroupSettings
   sortPriority=14
   texture=/Game/Effects/Fort_Effects/Textures/Tile/T_ShieldCaustics_Tile_Chromatic
  }
  n24 = Power(Base=ChannelMaskParameter(None=@n23.RGB) {group=Sequin, parameterName=SequinEmissiveGradientChannel, sortPriority=14}.R)
  n25 = Multiply(
   A=Multiply(A=Multiply(A=@n19, B=@n24), B=VectorParameter() {group=Sequin, parameterName=SparkleColor, defaultValue=<rgba 0.5742 0.228 1 1>}.RGB)
   B=ScalarParameter() {defaultValue=10, group=Sequin, parameterName=SparkleBRightness, sortPriority=13}
  )
  n26 = StaticSwitchParameter(True=Add(A=@n25, B=Multiply(A=@n13, B=@n14)), False=@n25) {group=Sequin, parameterName=MFSequin_UseThinFilmOnSequins}
  MaterialFunctionCall!(
   Low _ PC Low and Med=@n12
   "High _ XB1, PS4"=@n26
   "Epic _ PC High, XSX, PS5"=@n26
   --------
   "Mobile Low _ PC Perf, Android"=@n12
   Mobile Med _ Switch=@n12
  ) {materialFunction=/Game/Global/MaterialFunctions/TechArt/MF_Fortnite_QualitySwitch_Scalar}
  n27 = TextureCoordinate()
  n28 = Constant()
  n29 = FunctionInput(Preview=Constant()) {bUsePreviewValueAsDefault=true, inputName=SequinRotationAngle, inputType=FunctionInput_Scalar}
  NamedRerouteDeclaration(None=Multiply(
    A=MaterialFunctionCall(UVs=@n27, Rotation Center=@n28, "Rotation Angle (0-1)"=@n29) {materialFunction=&CustomRotator}.Rotated Values
    B=ScalarParameter() {defaultValue=32, group=Sequin, parameterName=SequinTile, sortPriority=5}
   )) {name=SequinUVs, nodeColor=<rgba 0 1 0.0859 1>}
  n30 = TextureSample!(
   UVs=NamedRerouteUsage()
   Tex=FunctionInput(Preview=TextureObject() {samplerType=SAMPLERTYPE_LinearGrayscale, texture=/Game/Global/Textures/Default/Tileable_Patterns/T_SequinTile}) {bUsePreviewValueAsDefault=true, inputName=SequinOffset, inputType=FunctionInput_Texture2D}
   Bias=ScalarParameter() {group=Sequin, parameterName=SequinTextureMipBias}
   Apply View MipBias
  ) {
   mipValueMode=TMVM_MipBias
   samplerSource=SSM_Wrap_WorldGroupSettings
   samplerType=SAMPLERTYPE_LinearGrayscale
   texture=&T_SequinTile
  }
  NamedRerouteDeclaration(None=@n30.RGB) {name=SequinOffset, nodeColor=<rgba 1 0 0.6562 1>}
  n31 = ScalarParameter() {group=Sequin, parameterName=SequinFresnel, sortPriority=7}
  n32 = OneMinus(None=@n31)
  n33 = Multiply(A=@n32, B=@n31)
  n34 = Multiply(
   A=ScalarParameter() {defaultValue=1.5, group=Sequin, parameterName=SequinBrightness}
   B=Multiply(A=NamedRerouteUsage(), B=LinearInterpolate(Alpha=@n33) {constA=0.7})
  )
  n35 = Multiply(A=NamedRerouteUsage(), B=ScalarParameter() {group=Sequin, parameterName=SequinThinFilmStrength_Basecolor, sortPriority=20})
  NamedRerouteDeclaration(None=StaticSwitchParameter(True=Add(A=@n34, B=@n35), False=@n34) {group=Sequin, parameterName=MFSequin_UseThinFilmOnSequins, sortPriority=20}) {name=BaseColor, nodeColor=<rgba 0 0.6719 1 1>}
  n36 = CameraVectorWS()
  n37 = Normalize(VectorInput=PixelNormalWS())
  n38 = Divide(A=ArccosineFast!(None=DotProduct(A=@n36, B=@n37))) {constB=1.5707}
  n39 = Multiply(
   A=Power(
    Base=NamedRerouteDeclaration(None=@n38) {name=LinearFresnel, nodeColor=<rgba 0.4531 0 1 1>}
    Exp=ScalarParameter() {defaultValue=1, group=Sequin, parameterName=SequinThinFilmUVExponent, sortPriority=17}
   )
   B=ScalarParameter() {defaultValue=1, group=Sequin, parameterName=SequinThinFilmUVScale, sortPriority=18}
  )
  n40 = ScalarParameter() {defaultValue=1, group=Sequin, parameterName=SequinThinFilmUVOffset, sortPriority=19}
  n41 = TextureSample!(
   UVs=MaterialFunctionCall(X=Add(A=@n39, B=@n40), Y=Constant() {r=0.5}) {materialFunction=&MakeFloat2}.Result
   Tex=FunctionInput(Preview=TextureObject() {texture=/Game/Global/Textures/Default/Thinfilm/T_ThinFilm_Spectrum_BlueGreen_Color}) {bUsePreviewValueAsDefault=true, inputName=SequinThinFilmTexture, inputType=FunctionInput_Texture2D}
   Apply View MipBias
  ) {samplerSource=SSM_Wrap_WorldGroupSettings, texture=&T_ThinFilm_Spectrum_BlueGreen_Color}
  n42 = ScalarParameter() {defaultValue=0.4, group=Sequin, parameterName=SequinThinFilmMaskThreshold, sortPriority=23}
  n43 = ScalarParameter() {defaultValue=0.1, group=Sequin, parameterName=SequinThinFilmMaskBias, sortPriority=24}
  n44 = Saturate(None=InverseLinearInterpolate(A=@n42, B=Add(A=@n42, B=@n43), Value=NamedRerouteUsage()))
  n45 = Multiply(A=@n44, B=@n44)
  NamedRerouteDeclaration(None=Multiply(
    A=NamedRerouteDeclaration(None=@n41.RGB) {name=SequinThinFilmColor, nodeColor=<rgba 1 0.7969 0 1>}
    B=NamedRerouteDeclaration(None=@n45) {name=SequinThinFilmMask, nodeColor=<rgba 0.5703 1 0 1>}
   )) {name=SequinThinFilm, nodeColor=<rgba 1 0 0.9141 1>}
  n46 = Constant() {r=0.35}
  n47 = TextureSample!(
   UVs=NamedRerouteUsage()
   Tex=FunctionInput(Preview=TextureObject() {samplerType=SAMPLERTYPE_LinearGrayscale, texture=/Game/Global/Textures/Default/Tileable_Patterns/T_SequinTile_roughness}) {bUsePreviewValueAsDefault=true, inputName=SequinRoughness, inputType=FunctionInput_Texture2D}
   Bias=ScalarParameter() {group=Sequin, parameterName=SequinTextureMipBias}
   Apply View MipBias
  ) {
   mipValueMode=TMVM_MipBias
   samplerSource=SSM_Wrap_WorldGroupSettings
   samplerType=SAMPLERTYPE_LinearGrayscale
   texture=&T_SequinTile_roughness
  }
  n48 = StaticSwitchParameter(
   True=LinearInterpolate(A=@n47.RGB, B=NamedRerouteUsage(), Alpha=ScalarParameter() {defaultValue=1, group=Sequin, parameterName=SequinBaseRoughnessBlendAmount})
   False=@n47.RGB
  ) {group=Sequin, parameterName=MFSequin_UseBaseRoughness}
  n49 = MaterialFunctionCall!(
   Low _ PC Low and Med=@n46
   "High _ XB1, PS4"=@n48
   "Epic _ PC High, XSX, PS5"=@n48
   --------
   "Mobile Low _ PC Perf, Android"=@n46
   Mobile Med _ Switch=@n46
  ) {materialFunction=/Game/Global/MaterialFunctions/TechArt/MF_Fortnite_QualitySwitch_Vector3}
  NamedRerouteDeclaration(None=@n49.Result) {name=Emissive_Out_1, nodeColor=<rgba 0.0312 1 0 1>}
  NamedRerouteDeclaration(None=@n2.Metallic) {name=BaseMetallic, nodeColor=<rgba 1 0.3984 0 1>}
  NamedRerouteDeclaration(None=LinearInterpolate(
    A=ScalarParameter() {defaultValue=0.9, parameterName=SequinMetalness}
    B=NamedRerouteUsage()
    Alpha=ScalarParameter() {defaultValue=1, group=Sequin, parameterName=SequinBaseRoughnessBlendAmount}
   )) {name=Metalness_Out, nodeColor=<rgba 0.0781 0 1 1>}
  n50 = VectorParameter() {group=Sequin, parameterName=SequinFallofColor01, defaultValue=<rgba 0.8842 0.6282 1 1>}
  n51 = VectorParameter() {group=Sequin, parameterName=SequinFallofColor02, defaultValue=<rgba 0.6663 0.575 1 1>}
  n52 = LinearInterpolate(Alpha=NamedRerouteUsage()) {constA=0.5, constB=1.5}
  NamedRerouteDeclaration(None=LinearInterpolate(
    A=FunctionInput(Preview=@n50.RGB) {bUsePreviewValueAsDefault=true, inputName=SequinFalloffColor01}
    B=FunctionInput(Preview=@n51.RGB) {bUsePreviewValueAsDefault=true, inputName=SequinFalloffColor02}
    Alpha=Multiply(A=@n52, B=@n32)
   )) {name=SequinColor, nodeColor=<rgba 0.8984 0 1 1>}
  NamedRerouteDeclaration(None=@n2.Roughness) {name=BaseRoughness, nodeColor=<rgba 0.6172 1 0 1>}
  NamedRerouteDeclaration(None=@n2.Normal) {name=BaseNormals, nodeColor=<rgba 0 1 0.1562 1>}
  n53 = NamedRerouteUsage()
  n54 = Multiply(
   A=ScalarParameter() {defaultValue=0.8, group=Sequin, parameterName=SequinNormalIntensity, sortPriority=15}
   B=Constant3Vector() {constant=<rgba 1 1 0 0>}
  )
  n55 = TextureSample(
   UVs=NamedRerouteUsage()
   Tex=FunctionInput(Preview=TextureObject() {samplerType=SAMPLERTYPE_Normal, texture=/Game/Global/Textures/Default/Tileable_Patterns/T_SequinTile_N}) {bUsePreviewValueAsDefault=true, inputName=SequinNormal, inputType=FunctionInput_Texture2D}
   Bias=ScalarParameter() {group=Sequin, parameterName=SequinTextureMipBias}
  ) {
   mipValueMode=TMVM_MipBias
   samplerSource=SSM_Wrap_WorldGroupSettings
   samplerType=SAMPLERTYPE_Normal
   texture=&T_SequinTile_N
  }
  n56 = Multiply(A=@n55.RGB, B=@n54)
  NamedRerouteDeclaration(None=StaticSwitchParameter(True=Multiply(A=@n53, B=@n54), False=Add(A=@n53, B=@n56) {constB=0}) {group=Sequin, parameterName=MFSequin_UseBaseNormal}) {name=Normal_Out, nodeColor=<rgba 0.2891 0 1 1>}
  NamedRerouteDeclaration(None=MaterialFunctionCall(
    Low _ PC Low and Med=@n12
    "High _ XB1, PS4"=@n26
    "Epic _ PC High, XSX, PS5"=@n26
    "Mobile Low _ PC Perf, Android"=@n12
    Mobile Med _ Switch=@n12
   ) {materialFunction=&MF_Fortnite_QualitySwitch_Vector3}.Result) {name=Emissive_Out, nodeColor=<rgba 0.0312 1 0 1>}
  NamedRerouteDeclaration(None=@n49.Result) {name=Roughness, nodeColor=<rgba 1 0 0.3516 1>}
  Transform(None=NamedRerouteUsage())
 ]}
