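# Parent body material for the TurnVortex Brink female character skin, built between the 2023 character-parent PreFX and PostFX functions.
# Low-end sequins, confined by an FX-texture channel and a local-space sphere mask, feed a UV2 composite.
# FX and TechArt masks combine with a TurnVortex gradient function, scaled by GlowReactive parameters, to add colored emissive while dimming base color where it glows.
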Material /BRCosmetics/Characters/Player/Female/Medium/Bodies/F_MED_TurnVortex_Brink/Materials/M_TurnVortex_Brink_Body_Parent
graph {
 properties {
  bUseMaterialAttributes=true
  bUsedWithMeshDeformer=true
  bUsedWithMorphTargets=true
  bUsedWithSkeletalMesh=true
  pixelDepthOffsetMode=PDOM_Legacy
  shadingModel=MSM_FromMaterialExpression
  usedShadingModels=Default Lit | Subsurface | Clear Coat | Subsurface Profile | Cloth
 }
 nodes [
  n1 = GetMaterialAttributes(None=NamedRerouteUsage()) {attributeGetTypes [#1, #2]}
  n2 = NamedRerouteUsage()
  n3 = TextureSample() {
   constCoordinate=1
   samplerSource=SSM_Wrap_WorldGroupSettings
   samplerType=SAMPLERTYPE_Masks
   texture=/BRCosmetics/Characters/Player/Female/Medium/Bodies/F_MED_TurnVortex_Brink/Textures/T_TurnVortex_Brink_Skin_FX
  }
  n4 = TextureSample() {
   samplerSource=SSM_Wrap_WorldGroupSettings
   samplerType=SAMPLERTYPE_Masks
   texture=&T_TurnVortex_Brink_Skin_FX
  }
  n5 = NamedRerouteUsage()
  n6 = Multiply(A=Max(A=@n3.R, B=@n4.G), B=Max(A=@n4.G, B=ComponentMask(None=@n5) {b=true}))
  n7 = MaterialFunctionCall() {materialFunction=&MF_CharacterParent_2023_PreFX}
  n8 = ComponentMask(None=@n7.M) {r=true}
  n9 = Max(A=@n6, B=Multiply(A=@n6, B=OneMinus(None=@n8)))
  n10 = MaterialFunctionCall() {materialFunction=&/MF_GradientColor_TurnVortex}
  n11 = Multiply(A=Multiply(A=ScalarParameter() {group=TurnVortex, parameterName=GlowReactive}, B=ScalarParameter() {defaultValue=1, parameterName=GlowReactiveSwitch})) {constB=0.5}
  n12 = Multiply(A=Multiply(A=@n9, B=@n10.Color), B=Multiply(A=@n10.Mask, B=@n11))
  n13 = NamedRerouteUsage()
  n14 = LinearInterpolate(
   B=ScalarParameter() {defaultValue=0.2, group=TurnVortex, parameterName=BaseColorBrightness}
   Alpha=Multiply(A=@n11, B=Multiply(A=@n9, B=@n10.Mask))
  ) {constA=1, constB=0.75}
  MaterialFunctionCall(
   Material=SetMaterialAttributes!(MaterialAttributes=@n1.MaterialAttributes, Emissive Color=Add(A=@n2, B=@n12), Base Color=Multiply(A=@n13, B=@n14)) {attributeSetTypes [#1, #2]}
   M=@n7.M
   Pre Skinned Local Position=@n7.M
   Distance Blend=@n7.M
   Fresnel=@n7.M
  ) {materialFunction=&MF_CharacterParent_2023_PostFX}
  NamedRerouteDeclaration(None=@n7.Tech Art Mask) {name=TechArtMask, nodeColor=<rgba 0.7578 1 0 1>}
  NamedRerouteDeclaration(None=@n1.Base Color) {name=InBaseColor, nodeColor=<rgba 0 0.5547 1 1>}
  NamedRerouteDeclaration(None=@n1.Emissive Color) {name=InEmissiveColor, nodeColor=<rgba 0 0.1562 1 1>}
  n15 = TextureObjectParameter() {
   group=Sequin
   parameterName=SequinRoughness
   samplerType=SAMPLERTYPE_LinearGrayscale
   sortPriority=0
   texture=/Game/Global/Textures/Default/Tileable_Patterns/T_SequinTile_roughness
  }
  n16 = TextureObjectParameter() {
   group=Sequin
   parameterName=SequinOffest
   samplerType=SAMPLERTYPE_LinearGrayscale
   sortPriority=1
   texture=/Game/Global/Textures/Default/Tileable_Patterns/T_SequinTile
  }
  n17 = TextureObjectParameter() {
   group=Sequin
   parameterName=SequinNormal
   samplerType=SAMPLERTYPE_Normal
   sortPriority=2
   texture=/Game/Global/Textures/Default/Tileable_Patterns/T_SequinTile_N
  }
  n18 = VectorParameter() {group=Sequin, parameterName=SequinFalloffColor01, sortPriority=3, defaultValue=<rgba 0.8542 0.387 0.5588 1>}
  n19 = VectorParameter() {group=Sequin, parameterName=SequinFalloffColor02, sortPriority=4, defaultValue=<rgba 0.3993 0.2651 0.8438 1>}
  n20 = ScalarParameter() {group=Sequin, parameterName=SequinRotationAngle}
  n21 = TextureObjectParameter() {group=Sequin, parameterName=SequinThinFilm, texture=/Game/Global/Textures/Default/Thinfilm/T_ThinFilm_Spectrum_BlueGreen_Color}
  n22 = TextureSample() {
   samplerSource=SSM_Wrap_WorldGroupSettings
   samplerType=SAMPLERTYPE_Masks
   texture=&T_TurnVortex_Brink_Skin_FX
  }
  n23 = Multiply(
   A=SphereMask(A=@n7.Pre skinned Local Position, B=Constant4Vector() {constant=<rgba 0 0 100.5305 1>}.RGB, Radius=Constant() {r=19.7853})
   B=@n22.B
  )
  NamedRerouteDeclaration(None=MaterialFunctionCall!(
    Material=MaterialFunctionCall!(
     inMat=@n7.Result
     SequinRoughness=@n15
     SequinOffset=@n16
     SequinNormal=@n17
     SequinFalloffColor01=@n18.RGB
     SequinFalloffColor02=@n19.RGB
     SequinRotationAngle=@n20
     SequinThinFilmTexture=@n21
     TechArtMask=@n23
    ) {materialFunction=&/MF_Sequins_Lowend}.Result
    TechArtMask=NamedRerouteUsage()
   ) {materialFunction=/Game/Global/Materials/Character_Parent_2023/UV2Composite_Parent_2023/MF_UV2_Composite_2023}.Result) {name=InMA, nodeColor=<rgba 0.4766 1 0 1>}
 ]
}
