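# Material function that renders the eyes of the TurnVortex Brink character, blending sclera, iris and pupil shading over an incoming material through a mask.
# Cornea refraction from index ratio, view vector and mid-plane depth offsets UVs; iris UVs rescale around centre for animation.
# Sphere masks and an exponent-based ellipse pupil drive colour, roughness, specular, metallic and emissive, with debug colours.
# Flattened, blended normals feed clear coat; masks and refracted UVs are also output.
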MaterialFunction /BRCosmetics/Characters/Player/Female/Medium/Bodies/F_MED_TurnVortex_Brink/Materials/MF_TurnVortex_Brink_Eyes
graph {nodes [
  n1 = TextureSample(UVs=NamedRerouteUsage(), Tex=NamedRerouteUsage()) {samplerSource=SSM_Clamp_WorldGroupSettings, samplerType=SAMPLERTYPE_Normal, texture=/Game/Characters/Shared/FNEye/Textures/T_Eye_Sphere_N}
  n2 = MaterialFunctionCall(
   Normal=@n1.RGB
   Flatness=FunctionInput() {bUsePreviewValueAsDefault=true, inputName=Eye_lris_N_Flatness, inputType=FunctionInput_Scalar, sortPriority=12}
  ) {materialFunction=&FlattenNormal}
  n3 = MaterialFunctionCall(
   Normal=NamedRerouteUsage()
   Flatness=FunctionInput() {bUsePreviewValueAsDefault=true, inputName=IN_Normal_Flatness, inputType=FunctionInput_Scalar, sortPriority=21}
  ) {materialFunction=&FlattenNormal}
  ClearCoatNormalCustomOutput!(None=NamedRerouteDeclaration(None=MaterialFunctionCall(BaseNormal=@n2.Result, AdditionalNormal=@n3.Result) {materialFunction=&BlendAngleCorrectedNormals}.Result) {name=OutputNormal, nodeColor=<rgba 0.1016 1 0 1>})
  Constant() {r=0.01}
  Constant() {r=1}
  Constant() {r=0.02}
  Constant() {r=0.5}
  Constant() {r=1}
  FunctionInput() {bUsePreviewValueAsDefault=true, inputName=------------------------------------_1, inputType=FunctionInput_Scalar, sortPriority=11}
  FunctionInput() {bUsePreviewValueAsDefault=true, inputName=------------------------------------_2, inputType=FunctionInput_Scalar, sortPriority=13}
  FunctionInput() {bUsePreviewValueAsDefault=true, inputName=------------------------------------_3, inputType=FunctionInput_Scalar, sortPriority=16}
  FunctionInput() {bUsePreviewValueAsDefault=true, inputName=------------------------------------_4, inputType=FunctionInput_Scalar, sortPriority=18}
  FunctionInput() {bUsePreviewValueAsDefault=true, inputName=------------------------------------_5, inputType=FunctionInput_Scalar, sortPriority=20}
  n4 = GetMaterialAttributes(None=FunctionInput() {bUsePreviewValueAsDefault=true, inputName=InputMaterial, inputType=FunctionInput_MaterialAttributes, sortPriority=6}) {attributeGetTypes [#1]}
  n5 = NamedRerouteUsage()
  n6 = FunctionInput() {bUsePreviewValueAsDefault=true, inputName=Eye_BaseColor_Brightness, sortPriority=21, previewValue=<1 0 0 1>}
  n7 = LinearInterpolate(A=Multiply(A=@n5, B=@n6), Alpha=NamedRerouteUsage()) {constB=0}
  n8 = NamedRerouteUsage()
  n9 = NamedRerouteUsage()
  n10 = NamedRerouteUsage()
  n11 = NamedRerouteUsage()
  n12 = NamedRerouteUsage()
  n13 = NamedRerouteUsage()
  n14 = OneMinus(None=NamedRerouteUsage())
  FunctionOutput(None=BlendMaterialAttributes(
    A=@n4.MaterialAttributes
    B=MakeMaterialAttributes!(
     BaseColor=@n7
     Metallic=@n8
     Specular=@n9
     Roughness=@n10
     Anisotropy
     EmissiveColor=@n11
     Opacity
     OpacityMask
     Normal=@n12
     Tangent
     WorldPositionOffset
     SubsurfaceColor
     ClearCoat
     ClearCoatRoughness
     AmbientOcclusion
     Refraction
     CustomizedUVs_0
     CustomizedUVs_1
     CustomizedUVs_2
     CustomizedUVs_3
     CustomizedUVs_4
     CustomizedUVs_5
     CustomizedUVs_6
     CustomizedUVs_7
     PixelDepthOffset
     ShadingModel
     Displacement
    )
    Alpha=Multiply(A=@n13, B=@n14)
   ) {vertexAttributeBlendType=UseA}) {sortPriority=3}
  FunctionOutput(None=NamedRerouteUsage()) {outputName=IrisMask, sortPriority=1}
  n15 = Abs(None=NamedRerouteUsage())
  n16 = NamedRerouteUsage()
  n17 = LinearInterpolate(Alpha=@n16) {constA=0.03}
  n18 = LinearInterpolate(A=Constant() {r=0.8}, Alpha=@n16) {constB=2}
  n19 = Abs(None=NamedRerouteUsage())
  n20 = LinearInterpolate(A=NamedRerouteUsage(), Alpha=Power(Base=NamedRerouteUsage()))
  n21 = Add(A=Power(Base=Divide(A=@n15, B=@n17), Exp=@n18), B=Power(Base=Divide(A=@n19, B=@n20), Exp=@n18))
  n22 = MaterialFunctionCall(In=OneMinus(None=@n21), Contrast=NamedRerouteUsage()) {materialFunction=&CheapContrast}
  FunctionOutput(None=NamedRerouteDeclaration(None=Saturate(None=@n22.Result)) {name=Eye_pupil_Mask, nodeColor=<rgba 0.5938 1 0 1>}) {outputName=PupilMask, sortPriority=2}
  n23 = NamedRerouteUsage()
  n24 = Constant() {r=0.5}
  n25 = NamedRerouteUsage()
  n26 = NamedRerouteUsage()
  FunctionOutput(None=NamedRerouteDeclaration(None=SphereMask(A=@n23, B=@n24, Radius=@n25, Hardness=@n26) {hardnessPercent=90}) {name=ScleraMask, nodeColor=<rgba 0.1484 1 0 1>}) {outputName=Sclera_Mask}
  n27 = NamedRerouteUsage()
  n28 = MaterialFunctionCall() {materialFunction=&TangentBasis}
  n29 = NamedRerouteUsage()
  n30 = DotProduct(A=@n28.X U, B=@n29)
  n31 = Normalize(VectorInput=Subtract(A=@n28.X U, B=Multiply(A=@n30, B=@n29)))
  n32 = CrossProduct(A=@n31, B=@n29)
  n33 = Multiply(
   A=Multiply(A=Constant() {r=0.02}, B=Constant2Vector() {g=1, r=-1})
   B=MaterialFunctionCall(X=DotProduct(A=@n31, B=@n27), Y=DotProduct(A=@n32, B=@n27)) {materialFunction=&MakeFloat2}.Result
  )
  n34 = TextureCoordinate()
  FunctionOutput(None=NamedRerouteDeclaration(None=Add(A=@n33, B=@n34)) {name=Refracted UVs, nodeColor=<rgba 0.6406 1 0 1>}) {outputName=Refracted UVs, sortPriority=10}
  n35 = SphereMask(A=NamedRerouteUsage(), B=AppendVector(A=NamedRerouteUsage(), B=NamedRerouteUsage()), Radius=NamedRerouteUsage()) {hardnessPercent=90}
  FunctionOutput(None=@n35) {outputName=HighLightMask, sortPriority=10}
  FunctionOutput(None=@n5) {outputName=BaseColor, sortPriority=1}
  n36 = TextureCoordinate()
  MaterialFunctionCall(UVs=@n36, Texture Scale=Constant() {r=0.26}) {materialFunction=&ScaleUVsByCenter}
  n37 = TextureCoordinate()
  MaterialFunctionCall(UVs=@n37, Texture Scale=Constant() {r=3}) {materialFunction=&ScaleUVsByCenter}
  Multiply(
   A=@n33
   B=SphereMask(A=TextureCoordinate(), B=NamedRerouteUsage(), Radius=Constant() {r=0.205}, Hardness=Constant() {r=0.95}) {hardnessPercent=90}
  )
  NamedRerouteDeclaration(None=FunctionInput() {bUsePreviewValueAsDefault=true, inputName=Eye_Sclera_UV_Position, inputType=FunctionInput_Vector2, sortPriority=10}) {name=Eye_Sclera_UV_Offset, nodeColor=<rgba 1 0.3984 0 1>}
  n38 = FunctionInput() {
   bUsePreviewValueAsDefault=true
   inputName=Eye_Sclera_UV_Scale_Offset
   inputType=FunctionInput_Vector2
   sortPriority=10
   previewValue=<0.73 0.73 0 1>
  }
  NamedRerouteDeclaration(None=@n38) {name=Eye_Sclera_UV_Scale_Offset, nodeColor=<rgba 0.0781 0 1 1>}
  NamedRerouteDeclaration(None=CameraVectorWS()) {name=Camera Vector, nodeColor=<rgba 1 0 1 1>}
  n39 = FunctionInput() {
   bUsePreviewValueAsDefault=true
   inputName=Eye_Iris_UV_Scale_Offset
   inputType=FunctionInput_Vector2
   sortPriority=12
   previewValue=<0.33 0.33 0 1>
  }
  NamedRerouteDeclaration(None=@n39) {name=Eye_Iris_UV_Scale_Offset, nodeColor=<rgba 0.1016 1 0 1>}
  n40 = FunctionInput() {bUsePreviewValueAsDefault=true, inputName=pupil_size, inputType=FunctionInput_Scalar, sortPriority=25, previewValue=<0.544 0 0 1>}
  NamedRerouteDeclaration(None=@n40) {name=Pupil size, nodeColor=<rgba 0 1 0.5078 1>}
  n41 = FunctionInput() {bUsePreviewValueAsDefault=true, inputName=pupil_shape, inputType=FunctionInput_Scalar, sortPriority=25}
  NamedRerouteDeclaration(None=Saturate(None=OneMinus(None=@n41))) {name=Pupil Dilation, nodeColor=<rgba 0 0.5312 1 1>}
  NamedRerouteDeclaration(None=PixelNormalWS()) {name=_Normal, nodeColor=<rgba 0 0 1 1>}
  NamedRerouteDeclaration(None=Divide(A=Constant() {desc=airIoR, r=1.0003}, B=Constant() {r=1.336}) {constA=1.0003}) {name=n, nodeColor=<rgba 0 1 0.2031 1>}
  n42 = NamedRerouteUsage()
  n43 = NamedRerouteUsage()
  NamedRerouteDeclaration(None=Multiply(A=NamedRerouteUsage(), B=DotProduct(A=@n42, B=@n43) {desc=facing}) {desc=w}) {name=w, nodeColor=<rgba 1 0.8297 0 1>}
  n44 = NamedRerouteUsage()
  n45 = NamedRerouteUsage()
  n46 = Multiply(A=Subtract(A=@n44, B=@n45), B=Add(A=@n44, B=@n45))
  NamedRerouteDeclaration(None=SquareRoot(None=Add(A=@n46)) {desc=k}) {name=k, nodeColor=<rgba 0.518 1 0 1>}
  n47 = NamedRerouteUsage()
  n48 = NamedRerouteUsage()
  n49 = Subtract(A=Multiply(A=Subtract(A=@n47, B=@n48), B=NamedRerouteUsage()), B=Multiply(A=NamedRerouteUsage(), B=NamedRerouteUsage())) {desc=t}
  NamedRerouteDeclaration(None=Multiply(A=Normalize(VectorInput=@n49)) {constB=-1}) {name=Eye Refraction Direction, nodeColor=<rgba 0 0.2882 1 1>}
  n50 = TextureSample(UVs=@n36, Tex=NamedRerouteUsage()) {samplerType=SAMPLERTYPE_Masks}
  n51 = FunctionInput() {
   bUsePreviewValueAsDefault=true
   inputName=Eye_Lris_Refraction_Depth
   inputType=FunctionInput_Scalar
   sortPriority=12
   previewValue=<0.5 0 0 1>
  }
  n52 = Multiply(A=@n50.R, B=@n51)
  n53 = DotProduct(A=CameraVectorWS(), B=NamedRerouteUsage())
  n54 = LinearInterpolate(Alpha=Multiply(A=@n53, B=@n53)) {constA=0.325}
  NamedRerouteDeclaration(None=Multiply(A=NamedRerouteUsage(), B=Divide(A=@n52, B=@n54) {desc=heightW / cosAlpha - scale distance to iris plane by facing ratio of camera and plane})) {name=Refraction Vector, nodeColor=<rgba 0.0781 1 0 1>}
  NamedRerouteDeclaration(None=FunctionInput(Preview=TextureObjectParameter() {parameterName=T_Eye_Sphere_N, samplerType=SAMPLERTYPE_Normal, texture=&FlatNormalBlue}) {bUsePreviewValueAsDefault=true, inputName=T_Eye_Sphere_N, inputType=FunctionInput_Texture2D, sortPriority=19}) {name=T_Eye_Sphere_N, nodeColor=<rgba 0.1484 1 0 1>}
  NamedRerouteDeclaration(None=@n4.Normal) {name=IN_Normal, nodeColor=<rgba 0 0.6953 1 1>}
  NamedRerouteDeclaration(None=FunctionInput(Preview=TextureObjectParameter() {parameterName=T_EyeSclera_D, texture=&T_BlacksRGB}) {bUsePreviewValueAsDefault=true, inputName=T_EyeSclera_D, inputType=FunctionInput_Texture2D, sortPriority=19}) {name=T_EyeSclera_D, nodeColor=<rgba 0.1484 1 0 1>}
  NamedRerouteDeclaration(None=FunctionInput(Preview=TextureObjectParameter() {parameterName=T_Eye_Iris, texture=&T_BlacksRGB}) {bUsePreviewValueAsDefault=true, inputName=T_Eye_Iris, inputType=FunctionInput_Texture2D, sortPriority=19}) {name=T_Eye_Iris, nodeColor=<rgba 0.1484 1 0 1>}
  n55 = MaterialFunctionCall(UVs=Add(A=TextureCoordinate(), B=NamedRerouteUsage()), Texture Scale=NamedRerouteUsage()) {materialFunction=&ScaleUVsByCenter}
  NamedRerouteDeclaration(None=@n55.UVs) {name=Eye_Sclera_UV_1, nodeColor=<rgba 0.2422 1 0 1>}
  NamedRerouteDeclaration(None=FunctionInput(Preview=TextureObjectParameter() {parameterName=T_EyeLirs2_S, samplerType=SAMPLERTYPE_Masks, texture=&T_BlackMask}) {bUsePreviewValueAsDefault=true, inputName=T_EyeLirs2_S, inputType=FunctionInput_Texture2D, sortPriority=19}) {name=T_EyeLirs2_S, nodeColor=<rgba 0.1484 1 0 1>}
  n56 = TextureObjectParameter() {
   parameterName=T_EyeMidPlaneDisplacement
   samplerType=SAMPLERTYPE_LinearColor
   texture=/Game/Global/Textures/Default/Blanks/T_BlackLinearColor
  }
  NamedRerouteDeclaration(None=FunctionInput(Preview=@n56) {bUsePreviewValueAsDefault=true, inputName=T_EyeMidPlaneDisplacement, inputType=FunctionInput_Texture2D, sortPriority=19}) {name=T_EyeMidPlaneDisplacement, nodeColor=<rgba 0.1484 1 0 1>}
  NamedRerouteDeclaration(None=@n21) {name=Eye_pupil_UV, nodeColor=<rgba 0 0.6953 1 1>}
  n57 = FunctionInput() {bUsePreviewValueAsDefault=true, inputName=Eye_Iris_Mask_Radius, inputType=FunctionInput_Scalar, sortPriority=12, previewValue=<0.485 0 0 1>}
  NamedRerouteDeclaration(None=@n57) {name=Eye_Iris_Mask_Radius, nodeColor=<rgba 1 0 0.8672 1>}
  NamedRerouteDeclaration(None=LinearInterpolate(
    B=@n35
    Alpha=FunctionInput() {bUsePreviewValueAsDefault=true, inputName=UseHighlight_Radius, inputType=FunctionInput_Scalar, sortPriority=17}
   ) {constAlpha=0}) {name=UVHighlight, nodeColor=<rgba 0.7812 0 1 1>}
  NamedRerouteDeclaration(None=MaterialFunctionCall(UVs=@n55.UVs, Texture Scale=NamedRerouteUsage()) {materialFunction=&ScaleUVsByCenter}.UVs) {name=Eye_Iris_UV_1, nodeColor=<rgba 1 0.9844 0 1>}
  n58 = TextureSample(UVs=NamedRerouteUsage(), Tex=NamedRerouteUsage()) {samplerType=SAMPLERTYPE_Masks}
  n59 = NamedRerouteUsage()
  NamedRerouteDeclaration(None=LinearInterpolate(A=LinearInterpolate(B=@n58.G, Alpha=@n59) {constA=1}, Alpha=NamedRerouteUsage()) {constB=0.2}) {name=EyeSpecular, nodeColor=<rgba 0.8516 1 0 1>}
  n60 = Constant() {r=0.75}
  n61 = Saturate(None=Subtract(A=NamedRerouteUsage(), B=NamedRerouteUsage()))
  n62 = FunctionInput() {bUsePreviewValueAsDefault=true, inputName=Global Metallic, inputType=FunctionInput_Scalar, sortPriority=21, previewValue=<1 0 0 1>}
  NamedRerouteDeclaration(None=Multiply(A=LinearInterpolate(B=@n60, Alpha=@n61) {constB=0.8}, B=@n62)) {name=EyeMetallic, nodeColor=<rgba 0.8281 0 1 1>}
  n63 = Constant() {r=0.1}
  n64 = SphereMask(A=NamedRerouteUsage(), B=Constant() {r=0.5}, Radius=NamedRerouteUsage(), Hardness=NamedRerouteUsage()) {hardnessPercent=90}
  n65 = NamedRerouteDeclaration(None=@n64) {name=Lris_Mask, nodeColor=<rgba 1 0 0.6562 1>}
  NamedRerouteDeclaration(None=LinearInterpolate(A=LinearInterpolate(A=@n63, B=@n58.R, Alpha=@n65), B=Constant() {r=0.05}, Alpha=NamedRerouteUsage())) {name=EyeRoughness, nodeColor=<rgba 0 1 0.5078 1>}
  n66 = FunctionInput() {bUsePreviewValueAsDefault=true, inputName=Eye_Iris_Mask_Hardness, inputType=FunctionInput_Scalar, sortPriority=12, previewValue=<0.93 0 0 1>}
  NamedRerouteDeclaration(None=@n66) {name=Eye_Iris_Mask_Hardness, nodeColor=<rgba 0 1 0.8125 1>}
  n67 = NamedRerouteUsage()
  n68 = FunctionInput() {bUsePreviewValueAsDefault=true, inputName=IN_IrisEdgeBrightness, inputType=FunctionInput_Scalar, sortPriority=12, previewValue=<1 0 0 1>}
  n69 = LinearInterpolate(A=NamedRerouteUsage(), B=Multiply(A=@n67, B=@n68), Alpha=NamedRerouteUsage()) {constB=0}
  n70 = NamedRerouteUsage()
  n71 = FunctionInput() {bUsePreviewValueAsDefault=true, inputName=Eye_Pupil_ColorTint, sortPriority=15, previewValue=<0.002 0.0019 0.001 1>}
  NamedRerouteDeclaration(None=LinearInterpolate(A=LinearInterpolate(A=@n69, B=@n67, Alpha=@n70), B=@n71, Alpha=NamedRerouteUsage()) {constB=0}) {name=BaseColor, nodeColor=<rgba 0.2188 0 1 1>}
  n72 = FunctionInput() {bUsePreviewValueAsDefault=true, inputName=Eye_Pupil_Size_Max, inputType=FunctionInput_Scalar, sortPriority=15, previewValue=<30 0 0 1>}
  NamedRerouteDeclaration(None=@n72) {name=Pupil_Size_Max, nodeColor=<rgba 1 0.4922 0 1>}
  n73 = TextureSample(UVs=@n37, Tex=NamedRerouteUsage()) {samplerSource=SSM_Clamp_WorldGroupSettings, samplerType=SAMPLERTYPE_Normal, texture=&T_Eye_Sphere_N}
  NamedRerouteDeclaration(None=Transform(None=@n73.RGB)) {name=EyeDirection, nodeColor=<rgba 1 0.3047 0 1>}
  n74 = FunctionInput() {bUsePreviewValueAsDefault=true, inputName=Eye_Pupil_Size_Min, inputType=FunctionInput_Scalar, sortPriority=15, previewValue=<6 0 0 1>}
  NamedRerouteDeclaration(None=@n74) {name=Pupil_Size_Min, nodeColor=<rgba 0.1719 0 1 1>}
  n75 = FunctionInput() {
   bUsePreviewValueAsDefault=true
   inputName=Eye_Pupil_VerticalDilation
   inputType=FunctionInput_Scalar
   sortPriority=15
   previewValue=<1.5 0 0 1>
  }
  NamedRerouteDeclaration(None=@n75) {name=Pupil_VerticalDilation, nodeColor=<rgba 0 1 0.7188 1>}
  n76 = FunctionInput() {bUsePreviewValueAsDefault=true, inputName=Eye_Pupil_Contrast, inputType=FunctionInput_Scalar, sortPriority=15, previewValue=<3 0 0 1>}
  NamedRerouteDeclaration(None=@n76) {name=Pupil_Contrast, nodeColor=<rgba 1 0.3984 0 1>}
  n77 = FunctionInput() {bUsePreviewValueAsDefault=true, inputName=UVHighlight_POS_X, inputType=FunctionInput_Scalar, sortPriority=17, previewValue=<0.414 0 0 1>}
  NamedRerouteDeclaration(None=@n77) {name=UVHighlight_POS_X, nodeColor=<rgba 0 1 0.5313 1>}
  n78 = FunctionInput() {bUsePreviewValueAsDefault=true, inputName=UVHighlight_POS_Y, inputType=FunctionInput_Scalar, sortPriority=17, previewValue=<0.454 0 0 1>}
  NamedRerouteDeclaration(None=@n78) {name=UVHighlight_POS_Y, nodeColor=<rgba 1 0.2109 0 1>}
  n79 = FunctionInput() {bUsePreviewValueAsDefault=true, inputName=UVHighlight_Radius, inputType=FunctionInput_Scalar, sortPriority=17, previewValue=<0.02 0 0 1>}
  NamedRerouteDeclaration(None=@n79) {name=UVHighlight_Radius, nodeColor=<rgba 0.5234 1 0 1>}
  n80 = FunctionInput() {
   bUsePreviewValueAsDefault=true
   inputName=Eye_Sclera_Mask_Radius
   inputType=FunctionInput_Scalar
   sortPriority=10
   previewValue=<0.485 0 0 1>
  }
  NamedRerouteDeclaration(None=@n80) {name=Eye_Sclera_Mask_Radius, nodeColor=<rgba 1 0 0.8437 1>}
  n81 = FunctionInput() {
   bUsePreviewValueAsDefault=true
   inputName=Eye_Sclera_Mask_Hardness
   inputType=FunctionInput_Scalar
   sortPriority=10
   previewValue=<0.93 0 0 1>
  }
  NamedRerouteDeclaration(None=@n81) {name=Eye_Sclera_Mask_Hardness, nodeColor=<rgba 0 1 0.8359 1>}
  n82 = NamedRerouteUsage()
  n83 = NamedRerouteUsage()
  n84 = NamedRerouteUsage()
  n85 = Constant() {r=0.389}
  n86 = NamedRerouteUsage()
  MaterialExpressionMultiply_13 = Multiply(
   A=Add(A=NamedRerouteUsage()) {constB=-0.5}
   B=Multiply(A=LinearInterpolate(A=@n82, B=@n83, Alpha=@n84) {constA=1, constB=0}, B=LinearInterpolate(B=@n85, Alpha=@n86) {constA=1})
  ) {constB=2}
  n87 = Convert(None=@MaterialExpressionMultiply_13) {
   nodeName=Break Float2
   convertInputs [{
     type=Vector2
     expressionInput {
      expression=&self:MaterialExpressionMultiply_13
      inputName=None
      mask=0
      maskA=0
      maskB=0
      maskG=0
      maskR=0
      outputIndex=0
     }
    }]
   convertOutputs [{type=Scalar}, {type=Scalar}]
  }
  NamedRerouteDeclaration(None=@n87) {name=L_pupil_UVx, nodeColor=<rgba 0.1719 1 0 1>}
  NamedRerouteDeclaration(None=FunctionInput() {bUsePreviewValueAsDefault=true, inputName=Eye_Debug, inputType=FunctionInput_Scalar, sortPriority=21}) {name=Eye_Debug, nodeColor=<rgba 0 0.5312 1 1>}
  n88 = NamedRerouteUsage()
  n89 = FunctionInput() {bUsePreviewValueAsDefault=true, inputName=Eye_Sclera_Emissive_Brightness, inputType=FunctionInput_Scalar, sortPriority=10}
  n90 = FunctionInput() {bUsePreviewValueAsDefault=true, inputName=Eye_lris_Emissive_Brightness, inputType=FunctionInput_Scalar, sortPriority=12}
  n91 = NamedRerouteUsage()
  n92 = Multiply(
   A=NamedRerouteUsage()
   B=LinearInterpolate(
    A=LinearInterpolate(A=@n89, B=@n90, Alpha=@n91)
    B=FunctionInput() {bUsePreviewValueAsDefault=true, inputName=Eye_Pupil_Emissive_Brightness, inputType=FunctionInput_Scalar, sortPriority=15}
    Alpha=NamedRerouteUsage()
   )
  )
  n93 = LinearInterpolate(
   A=LinearInterpolate(A=Constant3Vector() {constant=<rgba 1 1 1 0>}, B=Constant3Vector() {constant=<rgba 1 0 0 0>}, Alpha=NamedRerouteUsage())
   B=Constant3Vector() {constant=<rgba 0 1 0 0>}
   Alpha=NamedRerouteUsage()
  )
  n94 = Constant3Vector() {constant=<rgba 0 0 1 0>}
  n95 = NamedRerouteUsage()
  NamedRerouteDeclaration(None=LinearInterpolate(A=Add(A=@n88, B=@n92), B=LinearInterpolate(A=@n93, B=@n94, Alpha=@n95), Alpha=NamedRerouteUsage())) {name=EmissiveColor, nodeColor=<rgba 1 0 0.3281 1>}
  NamedRerouteDeclaration(None=Saturate(None=Multiply(A=@n64) {constB=10})) {name=Lris_Mask_EdgeHard, nodeColor=<rgba 1 0 0.6562 1>}
  NamedRerouteDeclaration(None=@n87) {name=L_pupil_UVy, nodeColor=<rgba 0.1719 1 0 1>}
  n96 = NamedRerouteUsage()
  n97 = NamedRerouteUsage()
  n98 = Power(Base=NamedRerouteUsage()) {constExponent=2.5}
  n99 = NamedRerouteUsage()
  n100 = Multiply(A=NamedRerouteUsage(), B=LinearInterpolate(A=Constant() {r=0.22}, Alpha=Power(Base=@n99) {constExponent=4}) {constA=1}) {constB=0.5}
  n101 = Power(Base=LinearInterpolate(Alpha=Length(None=@n100) {is=class MaterialExpressionLength}) {constB=0.4}, Exp=Constant() {r=0.5}) {constExponent=3}
  n102 = LinearInterpolate(
   A=LinearInterpolate(A=Constant() {r=0.524}, B=Constant() {r=3}, Alpha=Saturate(None=@n98))
   B=Constant() {r=1}
   Alpha=Saturate(None=Saturate(None=@n101))
  ) {constA=1}
  n103 = Divide(
   A=NamedRerouteDeclaration(None=MaterialFunctionCall(UVs=@n96, Texture Scale=@n97) {materialFunction=&ScaleUVsByCenter}.UVs) {name=Eye_Iris_UV, nodeColor=<rgba 1 0.9844 0 1>}
   B=@n102
  )
  n104 = Constant() {r=0.5}
  n105 = NamedRerouteDeclaration(None=Subtract(A=Add(A=@n103, B=@n104), B=Divide(A=@n104, B=@n102) {desc=Scale the center offset by the scaling factor}) {desc=Re offset to maintain center}) {name=Eye_Iris_anim_UV, nodeColor=<rgba 1 0 0.2109 1>}
  n106 = TextureSample(UVs=@n105, Tex=NamedRerouteUsage()) {samplerSource=SSM_Clamp_WorldGroupSettings}
  n107 = FunctionInput() {bUsePreviewValueAsDefault=true, inputName=Eye_lris_Color_Tint_A, sortPriority=12, previewValue=<1 1 1 1>}
  n108 = FunctionInput() {bUsePreviewValueAsDefault=true, inputName=Eye_lris_Color_Tint_B, sortPriority=12, previewValue=<1 1 1 1>}
  n109 = Constant() {r=1}
  n110 = FunctionInput() {bUsePreviewValueAsDefault=true, inputName=lris_Contrast, inputType=FunctionInput_Scalar, sortPriority=12, previewValue=<0 1 1 1>}
  n111 = LinearInterpolate(
   A=@n107
   B=@n108
   Alpha=MaterialFunctionCall(In=Desaturation(None=@n106.RGB, Fraction=@n109), Contrast=@n110) {materialFunction=&CheapContrast}.Result
  )
  n112 = FunctionInput() {bUsePreviewValueAsDefault=true, inputName=Use_lris_Color_Tint, inputType=FunctionInput_Scalar, sortPriority=12, previewValue=<0 1 1 1>}
  NamedRerouteDeclaration(None=LinearInterpolate(A=@n106.RGB, B=Saturate(None=@n111), Alpha=@n112)) {name=Eye_Iris, nodeColor=<rgba 1 0 0.1875 1>}
  n113 = FunctionInput() {bUsePreviewValueAsDefault=true, inputName=Eye Refraction Mix, inputType=FunctionInput_Scalar, sortPriority=21, previewValue=<1 0 0 1>}
  n114 = LinearInterpolate(A=TextureCoordinate(), B=NamedRerouteUsage(), Alpha=@n113)
  n115 = NamedRerouteUsage()
  n116 = NamedRerouteDeclaration(None=MaterialFunctionCall(UVs=Add(A=@n114, B=@n115), Texture Scale=NamedRerouteUsage()) {materialFunction=&ScaleUVsByCenter}.UVs) {name=Eye_Sclera_UV, nodeColor=<rgba 0.2422 1 0 1>}
  n117 = NamedRerouteUsage()
  n118 = FunctionInput() {bUsePreviewValueAsDefault=true, inputName=Eye_Sclera_Color_Tint, sortPriority=10, previewValue=<1 1 1 1>}
  NamedRerouteDeclaration(None=Multiply(A=TextureSample(UVs=@n116, Tex=@n117) {samplerSource=SSM_Clamp_WorldGroupSettings}.RGB, B=@n118)) {name=Eye_Sclera, nodeColor=<rgba 0 0.5078 1 1>}
  NamedRerouteUsage()
  NamedRerouteUsage()
  NamedRerouteUsage()
  Power(Base=@n50.R) {constExponent=0.5}
  n119 = TextureSample(UVs=@n105, Tex=NamedRerouteUsage()) {samplerSource=SSM_Clamp_WorldGroupSettings, samplerType=SAMPLERTYPE_Masks}
  Reroute(None=Saturate(None=Multiply(A=@n119.R) {constB=3}))
  ShadingModel()
 ]}
