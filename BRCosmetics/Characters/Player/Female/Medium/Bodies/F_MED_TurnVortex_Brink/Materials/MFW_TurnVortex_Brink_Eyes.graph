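# Wrapper material function that exposes grouped "CatEye" parameters and textures for the TurnVortex Brink character's eyes.
# It feeds them with input material attributes into MF_TurnVortex_Brink_Eyes, returning the result plus sclera, iris, pupil and highlight masks, base color and refracted UVs.
# Sclera and iris UV scales blend between per-axis and uniform values.
# An optional static switch offsets each eye's sclera separately, using a hard UV-V threshold mask or an input override.
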MaterialFunction /BRCosmetics/Characters/Player/Female/Medium/Bodies/F_MED_TurnVortex_Brink/Materials/MFW_TurnVortex_Brink_Eyes
graph {nodes [
  n1 = VectorParameter() {group=CatEye_02Sclera, parameterName=Eye_Sclera_Color_Tint, sortPriority=52, defaultValue=<rgba 1 1 1 1>}
  n2 = VectorParameter() {group=CatEye_03lris, parameterName=Eye_lris_Color_Tint_B, sortPriority=53, defaultValue=<rgba 1 1 1 1>}
  n3 = VectorParameter() {group=CatEye_04Pupils, parameterName=Eye_Pupil_ColorTint, sortPriority=54, defaultValue=<rgba 0.002 0.0019 0.002 1>}
  n4 = VectorParameter() {group=CatEye_03lris, parameterName=Eye_lris_Color_Tint_A, sortPriority=53, defaultValue=<rgba 1 1 1 1>}
  n5 = TextureObjectParameter() {
   group=CatEye_06Texture
   parameterName=T_EyeMidPlaneDisplacement
   samplerType=SAMPLERTYPE_LinearColor
   texture=/Game/Characters/Shared/FNEye/Textures/T_EyeMidPlaneDisplacement
  }
  n6 = TextureObjectParameter() {
   group=CatEye_06Texture
   parameterName=T_Eye_Sphere_N
   samplerType=SAMPLERTYPE_Normal
   texture=/Game/Characters/Shared/FNEye/Textures/T_Eye_Sphere_N
  }
  n7 = TextureObjectParameter() {
   group=CatEye_06Texture
   parameterName=T_EyeLirs2_S
   samplerType=SAMPLERTYPE_Masks
   texture=/Game/Characters/Shared/FNEye/Textures/T_Eye_lirs_S
  }
  n8 = MaterialFunctionCall!(
   InputMaterial=FunctionInput() {bUsePreviewValueAsDefault=true, inputName=IN_Material, inputType=FunctionInput_MaterialAttributes, sortPriority=6}
   Eye_Sclera_Mask_Radius=ScalarParameter() {defaultValue=0.485, group=CatEye_02Sclera, parameterName=Eye_Sclera_Mask_Radius, sortPriority=52}
   Eye_Sclera_UV_Scale_Offset=NamedRerouteUsage()
   Eye_Sclera_Mask_Hardness=ScalarParameter() {defaultValue=0.93, group=CatEye_02Sclera, parameterName=Eye_Sclera_Mask_Hardness, sortPriority=52}
   Eye_Sclera_Emissive_Brightness=ScalarParameter() {group=CatEye_02Sclera, parameterName=Eye_Sclera_Emissive_Brightness, sortPriority=52}
   Eye_Sclera_UV_Position=NamedRerouteUsage()
   Eye_Sclera_Color_Tint=@n1.RGB
   ------------------------------------_1
   Eye_Iris_Mask_Hardness=ScalarParameter() {defaultValue=0.93, group=CatEye_03lris, parameterName=Eye_Iris_Mask_Hardness, sortPriority=53}
   lris_Contrast=ScalarParameter() {group=CatEye_03lris, parameterName=Eye_lris_Contrast, sortPriority=53}
   Eye_Iris_Mask_Radius=ScalarParameter() {defaultValue=0.4, group=CatEye_03lris, parameterName=Eye_Iris_Mask_Radius, sortPriority=53}
   Eye_Iris_UV_Scale_Offset=NamedRerouteUsage()
   Eye_lris_N_Flatness=ScalarParameter() {group=CatEye_03lris, parameterName=Eye_lris_N_Flatness, sortPriority=53}
   IN_IrisEdgeBrightness=ScalarParameter() {defaultValue=1, group=CatEye_03lris, parameterName=Eye_Iris_EdgeBrightness, sortPriority=53}
   Eye_Lris_Refraction_Depth=ScalarParameter() {defaultValue=0.5, group=CatEye_03lris, parameterName=Eye_lris_Refraction_Depth, sortPriority=53}
   Use_lris_Color_Tint=ScalarParameter() {group=CatEye_03lris, parameterName=Use_lris_Color_Tint, sortPriority=53}
   Eye_lris_Emissive_Brightness=ScalarParameter() {group=CatEye_03lris, parameterName=Eye_lris_Emissive_Brightness, sortPriority=53}
   Eye_lris_Color_Tint_B=@n2.RGB
   Eye_lris_Color_Tint_A=@n4.RGB
   ------------------------------------_2
   Eye_Pupil_ColorTint=@n3.RGB
   Eye_Pupil_Contrast=ScalarParameter() {defaultValue=3, group=CatEye_04Pupils, parameterName=Eye_Pupil_Contrast, sortPriority=54}
   Eye_Pupil_VerticalDilation=ScalarParameter() {defaultValue=1.5, group=CatEye_04Pupils, parameterName=Eye_Pupil_VerticalDilation, sortPriority=54}
   Eye_Pupil_Size_Max=ScalarParameter() {defaultValue=30, group=CatEye_04Pupils, parameterName=Eye_Pupil_Size_Max, sortPriority=54}
   Eye_Pupil_Size_Min=ScalarParameter() {defaultValue=6, group=CatEye_04Pupils, parameterName=Eye_Pupil_Size_Min, sortPriority=54}
   Eye_Pupil_Emissive_Brightness=ScalarParameter() {group=CatEye_04Pupils, parameterName=Eye_Pupil_Emissive_Brightness, sortPriority=54}
   ------------------------------------_3
   UseHighlight_Radius=ScalarParameter() {group=CatEye_05HightLight, parameterName=UseHighlight, sortPriority=55}
   UVHighlight_POS_X=ScalarParameter() {defaultValue=0.414, group=CatEye_05HightLight, parameterName=UVHighlight_POS_X, sortPriority=55}
   UVHighlight_POS_Y=ScalarParameter() {defaultValue=0.454, group=CatEye_05HightLight, parameterName=UVHighlight_POS_Y, sortPriority=55}
   UVHighlight_Radius=ScalarParameter() {defaultValue=0.02, group=CatEye_05HightLight, parameterName=UVHighlight_Radius, sortPriority=55}
   ------------------------------------_4
   T_EyeSclera_D=TextureObjectParameter() {group=CatEye_06Texture, parameterName=T_EyeSclera_D, texture=/Game/Characters/Shared/FNEye/Textures/T_EyeSclera_D}
   T_Eye_Sphere_N=@n6
   T_EyeLirs2_S=@n7
   T_EyeMidPlaneDisplacement=@n5
   T_Eye_Iris=TextureObjectParameter() {group=CatEye_06Texture, parameterName=T_Eye_Iris, texture=/Game/Characters/Shared/FNEye/Textures/T_Eye_Iris_D}
   ------------------------------------_5
   Eye Refraction Mix=ScalarParameter() {defaultValue=1, group=CatEye_01Advance, parameterName=Eye_Refraction Mix, sortPriority=51}
   Eye_Debug=ScalarParameter() {group=CatEye_01Advance, parameterName=Eye_Debug_RGB, sortPriority=51}
   IN_Normal_Flatness=ScalarParameter() {group=CatEye_01Advance, parameterName=IN_Normal_Flatness, sortPriority=51}
   Eye_BaseColor_Brightness=ScalarParameter() {defaultValue=1, group=CatEye_01Advance, parameterName=Eye_BaseColor_Brightness, sortPriority=51}
   Global Metallic=ScalarParameter() {defaultValue=1, group=CatEye_01Advance, parameterName=Global Metallic, sortPriority=51}
   pupil_shape=ScalarParameter() {group=CatEye_04Pupils, parameterName=pupil_shape, sliderMax=1, sortPriority=54}
   pupil_size=ScalarParameter() {defaultValue=0.5448, group=CatEye_04Pupils, parameterName=pupil_size, sliderMax=1, sortPriority=54}
  ) {materialFunction=&/MF_TurnVortex_Brink_Eyes}
  FunctionOutput(None=@n8.Result) {outputName=Out_MA}
  FunctionOutput(None=@n8.Sclera_Mask) {outputName=Out_Sclera_Mask}
  FunctionOutput(None=@n8.IrisMask) {outputName=Out_lris_Mask}
  FunctionOutput(None=@n8.BaseColor) {outputName=Out_BaseColor}
  FunctionOutput(None=@n8.PupilMask) {outputName=Out_Pupil_Mask}
  FunctionOutput(None=@n8.HighLightMask) {outputName=Out_HighLightMask}
  FunctionOutput(None=@n8.Refracted UVs) {outputName=Out_Refracted_UV}
  n9 = AppendVector(
   A=ScalarParameter() {group=CatEye_02Sclera, parameterName=Eye_Sclera_UV_Position_X, sortPriority=52}
   B=ScalarParameter() {group=CatEye_02Sclera, parameterName=Eye_Sclera_UV_Position_Y, sortPriority=52}
  )
  n10 = AppendVector(
   A=ScalarParameter() {group=CatEye_02Sclera, parameterName=Eye_Sclera_UV_Position_X_R, sortPriority=52}
   B=ScalarParameter() {group=CatEye_02Sclera, parameterName=Eye_Sclera_UV_Position_Y_R, sortPriority=52}
  )
  n11 = ScalarParameter() {defaultValue=0.2, group=CatEye_02Sclera, parameterName=Eye_Sclera_UV_L_R_Mask, sortPriority=52}
  n12 = SmoothStep(Min=@n11, Max=@n11, Value=ComponentMask(None=TextureCoordinate()) {g=true})
  n13 = FunctionInput(Preview=NamedRerouteDeclaration(None=@n12) {name=L_R_Eye_mask, nodeColor=<rgba 1 0 0.7734 1>}) {bUsePreviewValueAsDefault=true, inputName=IN_Eye_Mask_L_R, inputType=FunctionInput_Scalar}
  NamedRerouteDeclaration(None=StaticSwitchParameter(True=LinearInterpolate(A=@n9, B=@n10, Alpha=@n13), False=@n9) {group=CatEye_01Advance, parameterName=Use Left eye and right eye, sortPriority=51}) {name=Eye_Sclera_UV_Position_In, nodeColor=<rgba 0.6172 1 0 1>}
  n14 = ScalarParameter() {defaultValue=1, group=CatEye_02Sclera, parameterName=Eye_Sclera_UV_Scale_X, sortPriority=52}
  n15 = ScalarParameter() {defaultValue=1, group=CatEye_02Sclera, parameterName=Eye_Sclera_UV_Scale_Y, sortPriority=52}
  NamedRerouteDeclaration(None=LinearInterpolate(
    A=AppendVector(A=@n14, B=@n15)
    B=ScalarParameter() {defaultValue=0.73, group=CatEye_02Sclera, parameterName=Eye_Sclera_UV_Scale_Uniform, sortPriority=52}
    Alpha=ScalarParameter() {defaultValue=1, group=CatEye_01Advance, parameterName=Use_Eye_UV_Uniform, sortPriority=51}
   )) {name=Eye_Sclera_UV_Scale_Offset_In, nodeColor=<rgba 0 1 0.7422 1>}
  n16 = ScalarParameter() {defaultValue=1, group=CatEye_03lris, parameterName=Eye_Iris_UV_Scale_X, sortPriority=53}
  n17 = ScalarParameter() {defaultValue=1, group=CatEye_03lris, parameterName=Eye_Iris_UV_Scale_Y, sortPriority=53}
  NamedRerouteDeclaration(None=LinearInterpolate(
    A=AppendVector(A=@n16, B=@n17)
    B=ScalarParameter() {defaultValue=0.33, group=CatEye_03lris, parameterName=Eye_Iris_UV_Scale_Uniform, sortPriority=53}
    Alpha=ScalarParameter() {defaultValue=1, group=01_Eye_Advance, parameterName=Use_Eye_UV_Uniform, sortPriority=51}
   )) {name=Eye_Iris_UV_Scale_Offset_In, nodeColor=<rgba 1 0.4219 0 1>}
  OneMinus(None=@n12)
 ]}
