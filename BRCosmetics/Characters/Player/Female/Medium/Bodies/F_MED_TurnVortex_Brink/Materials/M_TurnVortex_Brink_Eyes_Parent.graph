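# Parent eye material for the TurnVortex Brink female character body, using material attributes on skeletal meshes with Default Lit and Subsurface shading.
# The shared 2023 character eye function feeds the eyelash function, then tech-art overrides, forming the base look.
# A TurnVortex Brink eye function layers on that base, and the GlowReactive scalar blends from base to this character-specific result.
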Material /BRCosmetics/Characters/Player/Female/Medium/Bodies/F_MED_TurnVortex_Brink/Materials/M_TurnVortex_Brink_Eyes_Parent
graph {
 properties {
  bUseMaterialAttributes=true
  bUsedWithMeshDeformer=true
  bUsedWithSkeletalMesh=true
  pixelDepthOffsetMode=PDOM_Legacy
  shadingModel=MSM_FromMaterialExpression
  usedShadingModels=Default Lit | Subsurface
 }
 nodes [
  n1 = MaterialFunctionCall(In=MaterialFunctionCall!(Material=MaterialFunctionCall() {materialFunction=/Game/Global/Materials/Character_Parent_2023/Eye_Parent_2023/MF_eyes_2023}.Result) {materialFunction=/Game/Global/Materials/Character_Parent_2023/Eye_Parent_2023/MF_Eyelash}.Result) {materialFunction=&MF_TechArtOverrides}
  BlendMaterialAttributes(
   A=@n1.Result
   B=MaterialFunctionCall!(IN_Eye_Mask_L_R, IN_Material=@n1.Result) {materialFunction=&/MFW_TurnVortex_Brink_Eyes}.Out_MA
   Alpha=ScalarParameter() {group=TurnVortex_Brink, parameterName=GlowReactive}
  )
 ]
}
