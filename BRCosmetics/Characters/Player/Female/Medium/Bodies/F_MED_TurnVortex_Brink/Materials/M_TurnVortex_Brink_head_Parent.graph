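# Parent character material for the TurnVortex Brink female head, adding a reactive emissive vortex effect to the shared 2023 character shading.
# A gradient-color function, masked by the head FX texture's red channel, combines with a tiled starfield borrowed from another character.
# The gradient mask and GlowReactive scalars scale that glow.
# UseTurnVortexBrinkHead blends it over the PreFX function's attributes before the PostFX function finishes the material.
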Material /BRCosmetics/Characters/Player/Female/Medium/Bodies/F_MED_TurnVortex_Brink/Materials/M_TurnVortex_Brink_head_Parent
graph {
 properties {
  bUseMaterialAttributes=true
  bUsedWithMeshDeformer=true
  bUsedWithMorphTargets=true
  bUsedWithSkeletalMesh=true
  pixelDepthOffsetMode=PDOM_Legacy
  shadingModel=MSM_FromMaterialExpression
  usedShadingModels=Default Lit | Subsurface | Clear Coat | Subsurface Profile | Cloth
 }
 nodes [
  n1 = GetMaterialAttributes(None=NamedRerouteUsage()) {attributeGetTypes [#1, #2]}
  n2 = TextureSample() {
   samplerType=SAMPLERTYPE_Masks
   texture=/BRCosmetics/Characters/Player/Female/Medium/Bodies/F_MED_TurnVortex_Brink/Textures/T_TurnVortex_Brink_Head_FX
  }
  n3 = MaterialFunctionCall() {materialFunction=&/MF_GradientColor_TurnVortex}
  n4 = Multiply(A=@n2.R, B=@n3.Mask)
  n5 = TextureSample(UVs=Multiply(A=TextureCoordinate()) {constB=10}) {samplerType=SAMPLERTYPE_LinearColor, texture=/BRCosmetics/Characters/Player/Male/Medium/Bodies/M_MED_MarkerDeer_Tin/Textures/T_StarField}
  n6 = Multiply(A=ScalarParameter() {parameterName=GlowReactive}, B=ScalarParameter() {defaultValue=1, parameterName=GlowReactiveSwitch})
  n7 = Multiply(A=Add(A=Multiply(A=@n2.R, B=@n3.Color), B=Multiply(A=@n4, B=@n5.R)), B=Multiply(A=@n3.Mask, B=Multiply(A=@n6) {constB=0.5}))
  n8 = MaterialFunctionCall() {materialFunction=&MF_CharacterParent_2023_PreFX}
  MaterialFunctionCall(
   Material=BlendMaterialAttributes(
    A=NamedRerouteUsage()
    B=SetMaterialAttributes!(MaterialAttributes=@n1.MaterialAttributes, Emissive Color=@n7, Base Color) {attributeSetTypes [#1, #2]}
    Alpha=ScalarParameter() {parameterName=UseTurnVortexBrinkHead}
   )
   M=@n8.M
   Pre Skinned Local Position=@n8.M
   Distance Blend=@n8.M
   Fresnel=@n8.M
  ) {materialFunction=&MF_CharacterParent_2023_PostFX}
  NamedRerouteDeclaration(None=@n8.Result) {name=MA, nodeColor=<rgba 0.1484 1 0 1>}
 ]
}
