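# Control Rig for the F_MED_TurnVortex_Brink female body cosmetic, correcting its necklace and pendant on top of the shared DeformRig function.
# Constraints Setup parents base_necklace to weighted facial bones and base_pendant between spine_02 and pelvis.
# Leg Setup uses a left-thigh spherical pose reader to offset the pendant, with a rotated, shrunken variant when Crouching.
# An extra pendant rotation scales with remapped thigh-to-pelvis angle times smoothed DBNO state.
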ControlRigBlueprint /BRCosmetics/Characters/Player/Female/Medium/Bodies/F_MED_TurnVortex_Brink/Meshes/F_MED_TurnVortex_Brink_CtrlRig
graphs [
 New Function_ContainedGraph {nodes [
   RigVMFunctionEntryNode()
   > &RigUnit_ParentConstraint(
    AdvancedSettings=(RotationOrderForFilter=XZY)
    Child=(Type=Bone,Name="base_necklace")
    Parents="((Item=(Type=Bone,Name=\"FACIAL_R_UnderChin\"),Weight=0.0021),(Item=(Type=Bone,Name=\"FACIAL_L_UnderChin\"),Weight=0.0021),(Item=(Type=Bone,Name=\"FACIAL_C_UnderChin\"),Weight=0.021),(Item=(Type=Bone,Name=\"FACIAL_C_Jawline\"),Weight=0.0303),(Item=(Type=Bone,Name= ...(truncated)"
    Weight=1
   )
   > &RigUnit_ParentConstraint(
    AdvancedSettings=(RotationOrderForFilter=XZY)
    Child=(Type=Bone,Name="base_pendant")
    Parents=((Item=(Type=Bone,Name="spine_02"),Weight=0.456),(Item=(Type=Bone,Name="pelvis"),Weight=0.544))
    Weight=1
   )
   > RigVMFunctionReturnNode()
  ]}
 New Function_ContainedGraph {nodes [
   n1 = RigVMFunctionEntryNode()
   RigVMFunction_Sequence = &RigVMFunction_Sequence(ExecuteContext=@n1.ExecuteContext)
   n2 = &RigUnit_SphericalPoseReader(
    ActiveRegionScaleFactors=(PositiveWidth=1,NegativeWidth=1,PositiveHeight=1,NegativeHeight=1)
    ActiveRegionSize=0.05
    Debug=(bDrawDebug=false,bDraw2D=False,bDrawLocalAxes=False,DebugScale=25,DebugSegments=20,DebugThickness=0.25)
    DriverAxis=(X=-1,Y=0,Z=0)
    DriverItem=(Type=Bone,Name="thigh_l")
    FalloffRegionScaleFactors=(PositiveWidth=1,NegativeWidth=1,PositiveHeight=1,NegativeHeight=1)
    FalloffSize=0.45
    FlipHeightScaling=False
    FlipWidthScaling=false
    OptionalParentItem=(Type=Bone,Name="None")
    RotationOffset=(X=90,Y=0,Z=-20)
    ExecutePin=@RigVMFunction_Sequence.A
   )
   RigVMFunction_ControlFlowBranch = &RigVMFunction_ControlFlowBranch(ExecuteContext=@n2.ExecutePin, Condition=RigVMVariableNode() {variable=Crouching}.Value)
   &RigUnit_ModifyTransforms(
    ItemToModify=((Item=(Type=Bone,Name="base_pendant"),Transform=(Rotation=(X=-0.2257,Y=-0.4484,Z=0.1673,W=0.8486),Translation=(X=0,Y=0,Z=0),Scale3D=(X=1,Y=1,Z=1))))
    WeightMinimum=0
    ExecutePin=@RigVMFunction_ControlFlowBranch.False
    Weight=@n2.OutputParam
   )
   &RigUnit_ModifyTransforms(
    ItemToModify=((Item=(Type=Bone,Name="base_pendant"),Transform=(Rotation=(X=0.368,Y=0.2385,Z=-0.3662,W=-0.8207),Translation=(X=0,Y=0,Z=0),Scale3D=(X=0.7,Y=0.7,Z=0.7))))
    WeightMinimum=0
    ExecutePin=@RigVMFunction_ControlFlowBranch.True
    Weight=@n2.OutputParam
   )
   Delta Z = RigVMFunctionReferenceNode(Child=thigh_l, Max=-110, Min=-95, Parent=pelvis, ExecuteContext=@RigVMFunction_ControlFlowBranch.Completed)
   RigVMFunction_MathBoolToFloat = &RigVMFunction_MathBoolToFloat(Value=RigVMVariableNode() {variable=DBNO}.Value)
   AlphaInterp = &RigVMFunction_AlphaInterp(InterpSpeedDecreasing=10, InterpSpeedIncreasing=10, Scale=1, bClampResult=true, Value=@RigVMFunction_MathBoolToFloat.Result)
   Multiply = &RigVMFunction_MathDoubleMul(B=@AlphaInterp.Result, A=@"Delta Z".Delta Z)
   > &RigUnit_ModifyTransforms(
    ItemToModify=((Item=(Type=Bone,Name="base_pendant"),Transform=(Rotation=(X=0,Y=-0.3827,Z=0,W=0.9239),Translation=(X=0,Y=0,Z=0),Scale3D=(X=1,Y=1,Z=1))))
    WeightMinimum=0
    Weight=@Multiply.Result
   )
   RigVMFunctionReturnNode(ExecuteContext=@RigVMFunction_Sequence.B)
  ]}
 New Function_ContainedGraph {nodes [
   n1 = RigVMFunctionEntryNode()
   RigVMFunction_MathQuaternionToEuler = &RigVMFunction_MathQuaternionToEuler(
    RotationOrder=ZYX
    Value=&RigUnit_GetRelativeTransformForItem(
     Child=(Type=Bone,Name="None")
     Parent=(Type=Bone,Name="None")
     bChildInitial=False
     bParentInitial=False
     Child.Name=@n1.Child
     Parent.Name=@n1.Parent
    ).RelativeTransform.Rotation
   )
   Remap = &RigVMFunction_MathDoubleRemap(TargetMaximum=1, TargetMinimum=0, Value=@RigVMFunction_MathQuaternionToEuler.Result.Z, SourceMaximum=@n1.Max, SourceMinimum=@n1.Min)
   > RigVMFunctionReturnNode(Delta Z=@Remap.Result)
  ]}
 RigVMFunctionLibrary {nodes [
   Constraints Setup = RigVMCollapseNode(ExecuteContext=())
   Delta Z = RigVMCollapseNode(ExecuteContext=())
   Leg Setup = RigVMCollapseNode(ExecuteContext=())
  ]}
 RigVMModel {nodes [
   RigUnit_BeginExecution = &RigUnit_BeginExecution()
   FN_DeformRig_FNC = RigVMFunctionReferenceNode(
    DeformRig_ Curve="(deformRig_l_upperarm_up_alpha_72_#1=False,deformRig_r_upperarm_up_alpha_73_#2=False,deformRig_l_upperarm_fr_alpha_74_#3=False,deformRig_r_upperarm_fr_alpha_75_CB51B8 ...(truncated)"
    DeformRig_Bool="(upperarm_1_#4=True,elbow_3_#5=True,wrist_7_#6=True,groin_8_#7=True,glute_17_#8=True,knee_9_40C408324CA261 ...(truncated)"
    DeformRig_Debug=(upperarm_debug_18_#4=False,wrist_debug_20_#6=False,lat_debug_26_#9=False,groinGlute_debug_27_#7=False)
    DeformRig_Value="(upperarm_up_18_#10=(Rotation=(X=0,Y=0,Z=0,W=1),Translation=(X=0,Y=0,Z=8),Scale3D=(X=0.7,Y=1,Z=1)),upperarm_fwd_fr_19_#11=(Rotation=(X=0,Y=0,Z=0,W=1),Translation=(X=0,Y=10,Z=0),Scale3D=(X=0.5,Y=1,Z= ...(truncated)"
    ExecuteContext=@RigUnit_BeginExecution.ExecutePin
   )
   > Constraints Setup = RigVMFunctionReferenceNode()
   > Leg Setup = RigVMFunctionReferenceNode()
  ]}
]
hierarchy [
 root {loc=<0 0 0>, rot=<0 0 0 1>}
 pelvis {parent=root, loc=<0 -0.5301 98.0571>, rot=<0.0127 0.707 -0.0127 -0.707>}
 spine_01 {parent=pelvis, loc=<6.1208 -0.9328 0>, rot=<0 0 0.034 -0.9994>}
 spine_02 {parent=spine_01, loc=<6.5997 -0.2095 0>, rot=<0 0 0.0238 -0.9997>}
 spine_03 {parent=spine_02, loc=<6.4914 -0.1237 0>, rot=<0 0 0.0149 -0.9999>}
 spine_04 {parent=spine_03, loc=<6.9598 0.0794 0>, rot=<0 0 -0.0132 -0.9999>}
 spine_05 {parent=spine_04, loc=<6.9987 0 0>, rot=<0 0 -0.0015 -1>}
 neck_01 {parent=spine_05, loc=<11.3138 0.1529 0>, rot=<0 0 -0.2003 -0.9797>}
 neck_02 {parent=neck_01, loc=<4.5627 0 0>, rot=<0 0 0.0504 -0.9987>}
 head {parent=neck_02, loc=<5.3914 0 0>, rot=<0 0 0.115 -0.9934>}
 FACIAL_C_FacialRoot {parent=head, loc=<2.1524 -0.746 -0.0002>, rot=<0.4879 -0.5118 -0.4879 -0.5118>}
 FACIAL_C_UnderChin {parent=FACIAL_C_FacialRoot, loc=<0.0015 8.0522 4.3291>, rot=<0.4734 -0.0003 -0.0002 -0.8809>}
 FACIAL_L_UnderChin {parent=FACIAL_C_FacialRoot, loc=<2.3451 6.9234 2.9674>, rot=<0.3769 -0.3575 0.0714 -0.8515>}
 FACIAL_R_UnderChin {parent=FACIAL_C_FacialRoot, loc=<-2.349 6.9257 2.9602>, rot=<0.3769 0.3569 -0.0716 -0.8517>}
 FACIAL_C_Jaw {parent=FACIAL_C_FacialRoot, loc=<-0.0008 2.4782 1.9212>, rot=<0.3134 -0.0004 -0.0001 -0.9496>}
 FACIAL_C_Jawline {parent=FACIAL_C_Jaw, loc=<-0.0023 1.9183 7.0593>, rot=<0.2586 0 0 -0.966>}
 FACIAL_L_Ear {parent=FACIAL_C_FacialRoot, loc=<6.8474 -0.1517 -1.0088>, rot=<-0.0566 -0.0003 -0.0261 -0.9981>}
 FACIAL_R_Ear {parent=FACIAL_C_FacialRoot, loc=<-6.8566 -0.1588 -1.0166>, rot=<-0.0566 -0.0005 0.0262 -0.9981>}
 FACIAL_L_Jawline {parent=FACIAL_C_FacialRoot, loc=<4.0993 5.1894 4.2571>, rot=<0.18 0.159 -0.0499 -0.9695>}
 FACIAL_R_Jawline {parent=FACIAL_C_FacialRoot, loc=<-4.1054 5.1913 4.2605>, rot=<0.18 -0.1597 0.0498 -0.9693>}
 base_necklace {parent=neck_01, loc=<3.623 5.3903 0>, rot=<0.1257 0.6958 0.6958 0.1257>}
 dyn_necklace {parent=base_necklace, loc=<0 0 0>, rot=<0 0 0 1>}
 clavicle_l {parent=spine_05, loc=<5.4831 -0.1053 -3.2946>, rot=<0.0465 -0.7289 -0.009 -0.6829>}
 upperarm_l {parent=clavicle_l, loc=<9.8658 -1.2136 -0.4105>, rot=<0.0164 -0.3245 -0.0269 -0.9454>}
 lowerarm_l {parent=upperarm_l, loc=<26.465 -0.8345 0.259>, rot=<0.0036 0.0161 -0.2298 -0.9731>}
 hand_l {parent=lowerarm_l, loc=<22.7355 -0.3203 -0.6105>, rot=<-0.6784 -0.0223 0.0735 -0.7307>}
 thumb_01_l {parent=hand_l, loc=<1.3118 -0.3191 -2.459>, rot=<0.6546 -0.4042 0.0664 -0.6354>}
 thumb_02_l {parent=thumb_01_l, loc=<5.9336 -0.0448 0.0401>, rot=<-0.0037 0.0213 0.298 -0.9543>}
 thumb_03_l {parent=thumb_02_l, loc=<2.2223 0.3682 -0.0907>, rot=<0.0118 -0.013 0.0066 -0.9998>}
 index_metacarpal_l {parent=hand_l, loc=<1.6535 0.0183 -2.96>, rot=<0.0158 -0.0124 -0.0173 -0.9996>}
 index_01_l {parent=index_metacarpal_l, loc=<8.3008 0.0357 -0.0524>, rot=<-0.0044 0.0485 0.0924 -0.9945>}
 index_02_l {parent=index_01_l, loc=<3.3587 -0.3004 0.0452>, rot=<-0.002 0.0027 0.3225 -0.9466>}
 index_03_l {parent=index_02_l, loc=<2.5578 -0.123 -0.0252>, rot=<-0.002 0.0067 0.2558 -0.9667>}
 middle_metacarpal_l {parent=hand_l, loc=<2.1856 -0.1725 -0.5984>, rot=<0.006 0.0168 -0.0329 -0.9993>}
 middle_01_l {parent=middle_metacarpal_l, loc=<7.3275 -0.2447 0.0129>, rot=<-0.0208 0.0042 0.0858 -0.9961>}
 middle_02_l {parent=middle_01_l, loc=<4.0123 0.1401 0.1175>, rot=<-0.0218 0.0258 0.3512 -0.9357>}
 middle_03_l {parent=middle_02_l, loc=<2.6077 0.0534 0.1373>, rot=<0.0082 0.0188 0.1708 -0.9851>}
 ring_metacarpal_l {parent=hand_l, loc=<2.2074 -0.3625 1.3638>, rot=<-0.0284 0.0722 -0.0458 -0.9959>}
 ring_01_l {parent=ring_metacarpal_l, loc=<7.0295 -0.6197 -0.1446>, rot=<-0.0489 -0.0252 0.1892 -0.9804>}
 ring_02_l {parent=ring_01_l, loc=<3.6174 0.4647 -0.1027>, rot=<0 0 0.2806 -0.9598>}
 ring_03_l {parent=ring_02_l, loc=<2.4152 0.114 0.0199>, rot=<0 0 0.1189 -0.9929>}
 pinky_metacarpal_l {parent=hand_l, loc=<1.7616 -0.9753 3.3814>, rot=<-0.0397 0.0814 -0.069 -0.9935>}
 pinky_01_l {parent=pinky_metacarpal_l, loc=<7.1709 -0.4831 -0.1094>, rot=<-0.0654 -0.0035 0.2242 -0.9723>}
 pinky_02_l {parent=pinky_01_l, loc=<2.8274 0.0893 -0.1043>, rot=<0.0037 0.0011 0.2854 -0.9584>}
 pinky_03_l {parent=pinky_02_l, loc=<2.2046 0.0307 -0.0078>, rot=<-0.0002 0.0098 0.1445 -0.9895>}
 deform_wrist_dn_base_l {parent=hand_l, loc=<0 -1.5326 0>, rot=<0 0 0 -1>}
 deform_wrist_dn_l {parent=deform_wrist_dn_base_l, loc=<0 0 0>, rot=<0 0 0 1>}
 deform_wrist_up_base_l {parent=hand_l, loc=<0 2.3009 0>, rot=<0 0 0 -1>}
 deform_wrist_up_l {parent=deform_wrist_up_base_l, loc=<0 0 0>, rot=<0 0 0 1>}
 lowerarm_twist_02_l {parent=lowerarm_l, loc=<10.6707 0 0>, rot=<0 0 0 1>}
 lowerarm_twist_01_l {parent=lowerarm_l, loc=<18.316 0 0>, rot=<0 0 0 1>}
 deform_elbow_in_base_l {parent=lowerarm_l, loc=<0.1834 3.6062 -0.2572>, rot=<0 0 0 -1>}
 deform_elbow_in_l {parent=deform_elbow_in_base_l, loc=<0 0 0>, rot=<0 0 0 1>}
 deform_elbow_out_base_l {parent=lowerarm_l, loc=<0 -4 0>, rot=<0 0 0 1>}
 deform_elbow_out_l {parent=deform_elbow_out_base_l, loc=<0 0 0>, rot=<0 0 0 1>}
 upperarm_twist_01_l {parent=upperarm_l, loc=<7.2135 -0.1134 0>, rot=<0 0 0 1>}
 deform_upperarm_bk_base_l {parent=upperarm_twist_01_l, loc=<-6.4053 -4.3746 0.314>, rot=<0 0 0 -1>}
 deform_upperarm_bk_l {parent=deform_upperarm_bk_base_l, loc=<0 0 0>, rot=<0 0 0 1>}
 deform_upperarm_fr_base_l {parent=upperarm_twist_01_l, loc=<-5.6312 5.092 0.3354>, rot=<0 0 0 -1>}
 deform_upperarm_fr_l {parent=deform_upperarm_fr_base_l, loc=<0 0 0>, rot=<0 0 0 1>}
 deform_upperarm_up_base_l {parent=upperarm_twist_01_l, loc=<-8.891 0.4825 3.9517>, rot=<0 0 0 -1>}
 deform_upperarm_up_l {parent=deform_upperarm_up_base_l, loc=<0 0 0>, rot=<0 0 0 1>}
 upperarm_twist_02_l {parent=upperarm_l, loc=<19.7998 -0.1134 0>, rot=<0 0 0 1>}
 clavicle_r {parent=spine_05, loc=<5.4831 -0.1053 3.2946>, rot=<0.7289 0.0465 -0.6829 0.009>}
 upperarm_r {parent=clavicle_r, loc=<-9.8658 1.2136 0.4105>, rot=<0.0164 -0.3245 -0.0269 -0.9454>}
 lowerarm_r {parent=upperarm_r, loc=<-26.465 0.8345 -0.259>, rot=<0.0036 0.0161 -0.2298 -0.9731>}
 hand_r {parent=lowerarm_r, loc=<-22.7355 0.3203 0.6106>, rot=<-0.6784 -0.0223 0.0735 -0.7307>}
 thumb_01_r {parent=hand_r, loc=<-1.3118 0.3191 2.459>, rot=<0.6546 -0.4042 0.0664 -0.6354>}
 thumb_02_r {parent=thumb_01_r, loc=<-5.9336 0.0448 -0.0401>, rot=<-0.0037 0.0213 0.298 -0.9543>}
 thumb_03_r {parent=thumb_02_r, loc=<-2.2223 -0.3682 0.0907>, rot=<0.0118 -0.013 0.0066 -0.9998>}
 index_metacarpal_r {parent=hand_r, loc=<-1.6535 -0.0183 2.96>, rot=<0.0158 -0.0124 -0.0173 -0.9996>}
 index_01_r {parent=index_metacarpal_r, loc=<-8.3008 -0.0357 0.0524>, rot=<-0.0044 0.0485 0.0924 -0.9945>}
 index_02_r {parent=index_01_r, loc=<-3.3587 0.3004 -0.0452>, rot=<-0.002 0.0027 0.3225 -0.9466>}
 index_03_r {parent=index_02_r, loc=<-2.5578 0.123 0.0252>, rot=<-0.002 0.0067 0.2558 -0.9667>}
 middle_metacarpal_r {parent=hand_r, loc=<-2.1856 0.1725 0.5984>, rot=<0.006 0.0168 -0.0329 -0.9993>}
 middle_01_r {parent=middle_metacarpal_r, loc=<-7.3275 0.2447 -0.0129>, rot=<-0.0208 0.0042 0.0858 -0.9961>}
 middle_02_r {parent=middle_01_r, loc=<-4.0123 -0.1401 -0.1175>, rot=<-0.0218 0.0258 0.3512 -0.9357>}
 middle_03_r {parent=middle_02_r, loc=<-2.6077 -0.0534 -0.1373>, rot=<0.0082 0.0188 0.1708 -0.9851>}
 ring_metacarpal_r {parent=hand_r, loc=<-2.2074 0.3625 -1.3638>, rot=<-0.0284 0.0722 -0.0458 -0.9959>}
 ring_01_r {parent=ring_metacarpal_r, loc=<-7.0295 0.6197 0.1446>, rot=<-0.0489 -0.0252 0.1892 -0.9804>}
 ring_02_r {parent=ring_01_r, loc=<-3.6174 -0.4647 0.1027>, rot=<0 0 0.2806 -0.9598>}
 ring_03_r {parent=ring_02_r, loc=<-2.4152 -0.114 -0.0199>, rot=<0 0 0.1189 -0.9929>}
 pinky_metacarpal_r {parent=hand_r, loc=<-1.7616 0.9753 -3.3814>, rot=<-0.0397 0.0814 -0.069 -0.9935>}
 pinky_01_r {parent=pinky_metacarpal_r, loc=<-7.1709 0.4831 0.1094>, rot=<-0.0654 -0.0035 0.2242 -0.9723>}
 pinky_02_r {parent=pinky_01_r, loc=<-2.8274 -0.0893 0.1043>, rot=<0.0037 0.0011 0.2854 -0.9584>}
 pinky_03_r {parent=pinky_02_r, loc=<-2.2046 -0.0307 0.0078>, rot=<-0.0002 0.0098 0.1445 -0.9895>}
 deform_wrist_dn_base_r {parent=hand_r, loc=<-0.0001 1.5328 0>, rot=<0 0 0 -1>}
 deform_wrist_dn_r {parent=deform_wrist_dn_base_r, loc=<0 0 0>, rot=<0 0 0 1>}
 deform_wrist_up_base_r {parent=hand_r, loc=<-0.0002 -2.3007 -0.0001>, rot=<0 0 0 -1>}
 deform_wrist_up_r {parent=deform_wrist_up_base_r, loc=<0 0 0>, rot=<0 0 0 1>}
 lowerarm_twist_02_r {parent=lowerarm_r, loc=<-10.6707 0 0>, rot=<0 0 0 -1>}
 lowerarm_twist_01_r {parent=lowerarm_r, loc=<-18.316 0 0>, rot=<0 0 0 -1>}
 deform_elbow_in_base_r {parent=lowerarm_r, loc=<-0.1834 -3.6062 0.2572>, rot=<0 0 0 -1>}
 deform_elbow_in_r {parent=deform_elbow_in_base_r, loc=<0 0 0>, rot=<0 0 0 1>}
 deform_elbow_out_base_r {parent=lowerarm_r, loc=<0 4 -0.0001>, rot=<0 0 0 -1>}
 deform_elbow_out_r {parent=deform_elbow_out_base_r, loc=<0 0 0>, rot=<0 0 0 1>}
 upperarm_twist_01_r {parent=upperarm_r, loc=<-7.2135 0.1134 0>, rot=<0 0 0 -1>}
 deform_upperarm_bk_base_r {parent=upperarm_twist_01_r, loc=<6.4053 4.3746 -0.314>, rot=<0 0 0 -1>}
 deform_upperarm_bk_r {parent=deform_upperarm_bk_base_r, loc=<0 0 0>, rot=<0 0 0 1>}
 deform_upperarm_fr_base_r {parent=upperarm_twist_01_r, loc=<5.6312 -5.092 -0.3354>, rot=<0 0 0 -1>}
 deform_upperarm_fr_r {parent=deform_upperarm_fr_base_r, loc=<0 0 0>, rot=<0 0 0 1>}
 deform_upperarm_up_base_r {parent=upperarm_twist_01_r, loc=<8.891 -0.4825 -3.9517>, rot=<0 0 0 -1>}
 deform_upperarm_up_r {parent=deform_upperarm_up_base_r, loc=<0 0 0>, rot=<0 0 0 1>}
 upperarm_twist_02_r {parent=upperarm_r, loc=<-19.7998 0.1134 0>, rot=<0 0 0 -1>}
 attach_backpack {parent=spine_05, loc=<-0.1223 -13.4732 0>, rot=<-0.0093 -0.707 0.0093 -0.707>}
 Backpack_BR {parent=attach_backpack, type=Null, loc=<0 0 0>, rot=<0 0 0 1>}
 attach_cape {parent=spine_05, loc=<9.0174 -11.7141 0>, rot=<0.0616 -0.7044 -0.0616 -0.7044>}
 cape {parent=attach_cape, type=Null, loc=<0 0 0>, rot=<0 0 0 1>}
 chest_aim {parent=spine_05, loc=<8.2427 6.6249 0>, rot=<0.4795 0.5196 0.5196 0.4795>}
 deform_pec_aim_base_l {parent=spine_04, loc=<3.2456 9.3309 -1.2237>, rot=<0.0603 -0.4033 0.3318 -0.8507>}
 deform_pec_aim_l {parent=deform_pec_aim_base_l, loc=<0 0 0>, rot=<0 0 0 1>}
 deform_pec_base_l {parent=deform_pec_aim_l, loc=<9.3665 0 0>, rot=<0 0 0 1>}
 deform_pec_l {parent=deform_pec_base_l, loc=<0 0 0>, rot=<0 0 0 1>}
 deform_pec_aim_base_r {parent=spine_04, loc=<3.2452 9.3309 1.2237>, rot=<-0.4033 -0.0603 0.8507 0.3318>}
 deform_pec_aim_r {parent=deform_pec_aim_base_r, loc=<0 0 0>, rot=<0 0 0 1>}
 deform_pec_base_r {parent=deform_pec_aim_r, loc=<-9.3667 -0.0003 -0.0004>, rot=<0 0 0 1>}
 deform_pec_r {parent=deform_pec_base_r, loc=<0 0 0>, rot=<0 0 0 1>}
 thigh_l {parent=pelvis, loc=<-1.0855 -0.2109 -8.7053>, rot=<0.0009 0.0355 0.0155 -0.9993>}
 calf_l {parent=thigh_l, loc=<-42.5284 -0.0019 -0.7297>, rot=<0.0006 -0.0161 -0.0391 -0.9991>}
 foot_l {parent=calf_l, loc=<-46.7401 -2.5014 -0.592>, rot=<0.0427 -0.0238 0.0353 -0.9982>}
 ball_l {parent=foot_l, loc=<-3.5658 10.8066 -0.1544>, rot=<0.0001 -0.0001 -0.7063 -0.7079>}
 calf_twist_02_l {parent=calf_l, loc=<-19.1546 -1.7943 -0.2638>, rot=<0.0237 0.0095 -0.0035 -0.9997>}
 calf_twist_01_l {parent=calf_l, loc=<-35.8483 -1.9197 -0.5794>, rot=<0.0237 0.0095 -0.0035 -0.9997>}
 deform_knee_in_base_l {parent=calf_l, loc=<0.8031 -5.0231 0.2456>, rot=<0 0 0 -1>}
 deform_knee_in_l {parent=deform_knee_in_base_l, loc=<0 0 0>, rot=<0 0 0 1>}
 deform_knee_out_base_l {parent=calf_l, loc=<0 4 0>, rot=<0 0 0 -1>}
 deform_knee_out_l {parent=deform_knee_out_base_l, loc=<0 0 0>, rot=<0 0 0 1>}
 base_laces_l {parent=calf_l, loc=<-9.1695 2.7481 -0.5384>, rot=<0.043 0.7047 0.708 0.0156>}
 Base_rounfLaces_02_l {parent=base_laces_l, loc=<0.1 1.0196 -0.0429>, rot=<-0.7059 -0.041 -0.041 -0.7059>}
 dyn_rounfLaces_02_l {parent=Base_rounfLaces_02_l, loc=<0 0 0>, rot=<0 0 0 1>}
 base_rounfLaces_01_l {parent=base_laces_l, loc=<0.1 -1.0114 -0.0429>, rot=<-0.65 0.2783 0.2783 -0.65>}
 dyn_rounfLaces_01_l {parent=base_rounfLaces_01_l, loc=<0 0 0>, rot=<0 0 0 1>}
 base_lace_01_l {parent=base_laces_l, loc=<0.9141 -0.7571 -0.1956>, rot=<-0.7068 0.0191 0.0191 -0.7068>}
 dyn_lace_01_l {parent=base_lace_01_l, loc=<0 0 0>, rot=<0 0 0 1>}
 base_lace_02_l {parent=base_laces_l, loc=<0.9957 0.7479 -0.2908>, rot=<-0.7068 0.0191 0.0191 -0.7068>}
 dyn_lace_02_l {parent=base_lace_02_l, loc=<0 0 0>, rot=<0 0 0 1>}
 base_zipper_l {parent=calf_l, loc=<-8.0626 -2.8006 3.6422>, rot=<0.043 0.7047 0.708 0.0156>}
 dyn_zipper_l {parent=base_zipper_l, loc=<0 0 0>, rot=<0 0 0 1>}
 thigh_twist_01_l {parent=thigh_l, loc=<-27.952 0 0>, rot=<0.0013 0 0 -1>}
 deform_groin_base_l {parent=thigh_l, loc=<0.4801 5.6776 -0.2486>, rot=<0 0 0 -1>}
 deform_groin_l {parent=deform_groin_base_l, loc=<0 0 0>, rot=<0 0 0 1>}
 thigh_r {parent=pelvis, loc=<-1.0855 -0.2109 8.7053>, rot=<0.0355 -0.0009 0.9993 0.0155>}
 calf_r {parent=thigh_r, loc=<42.5284 0.0019 0.7297>, rot=<0.0006 -0.0161 -0.0391 -0.9991>}
 foot_r {parent=calf_r, loc=<46.7401 2.5014 0.592>, rot=<0.0427 -0.0238 0.0353 -0.9982>}
 ball_r {parent=foot_r, loc=<3.5658 -10.8066 0.1544>, rot=<0.0001 -0.0001 -0.7063 -0.7079>}
 calf_twist_02_r {parent=calf_r, loc=<19.1546 1.7943 0.2638>, rot=<0.0237 0.0095 -0.0035 -0.9997>}
 calf_twist_01_r {parent=calf_r, loc=<35.8483 1.9197 0.5794>, rot=<0.0237 0.0095 -0.0035 -0.9997>}
 deform_knee_in_base_r {parent=calf_r, loc=<-0.8031 5.0231 -0.2457>, rot=<0 0 0 -1>}
 deform_knee_in_r {parent=deform_knee_in_base_r, loc=<0 0 0>, rot=<0 0 0 1>}
 deform_knee_out_base_r {parent=calf_r, loc=<0 -4 0>, rot=<0 0 0 -1>}
 deform_knee_out_r {parent=deform_knee_out_base_r, loc=<0 0 0>, rot=<0 0 0 1>}
 base_races_r {parent=calf_r, loc=<9.1694 -2.7481 0.5384>, rot=<0.043 0.7047 0.708 0.0156>}
 Base_rounfLaces_02_r {parent=base_races_r, loc=<-0.1001 -1.0196 0.0429>, rot=<-0.7059 -0.041 -0.041 -0.7059>}
 dyn_rounfLaces_02_r {parent=Base_rounfLaces_02_r, loc=<0 0 0>, rot=<0 0 0 1>}
 base_rounfLaces_01_r {parent=base_races_r, loc=<-0.1001 1.0114 0.0429>, rot=<-0.65 0.2783 0.2783 -0.65>}
 dyn_rounfLaces_01_r {parent=base_rounfLaces_01_r, loc=<0 0 0>, rot=<0 0 0 1>}
 base_race_01_r {parent=base_races_r, loc=<-0.9142 0.7571 0.1956>, rot=<-0.7068 0.0191 0.0191 -0.7068>}
 dyn_race_01_r {parent=base_race_01_r, loc=<0 0 0>, rot=<0 0 0 1>}
 base_race_02_r {parent=base_races_r, loc=<-0.9957 -0.7479 0.2908>, rot=<-0.7068 0.0191 0.0191 -0.7068>}
 dyn_race_02_r {parent=base_race_02_r, loc=<0 0 0>, rot=<0 0 0 1>}
 base_zipper_r {parent=calf_r, loc=<8.0626 2.8006 -3.6422>, rot=<0.043 0.7047 0.708 0.0156>}
 dyn_zipper_r {parent=base_zipper_r, loc=<0 0 0>, rot=<0 0 0 1>}
 thigh_twist_01_r {parent=thigh_r, loc=<27.952 0 0>, rot=<0.0013 0 0 -1>}
 deform_groin_base_r {parent=thigh_r, loc=<-0.4801 -5.6776 0.2486>, rot=<0 0 0 -1>}
 deform_groin_r {parent=deform_groin_base_r, loc=<0 0 0>, rot=<0 0 0 1>}
 deform_glute_base_l {parent=pelvis, loc=<-7.3466 -9.0213 -9.1245>, rot=<0.0172 0.031 -0.4477 -0.8935>}
 deform_glute_l {parent=deform_glute_base_l, loc=<0 0 0>, rot=<0 0 0 1>}
 deform_glute_base_r {parent=pelvis, loc=<-7.3464 -9.0213 9.1245>, rot=<0.031 -0.0172 0.8935 -0.4477>}
 deform_glute_r {parent=deform_glute_base_r, loc=<0 0 0>, rot=<0 0 0 1>}
 base_pendant {parent=pelvis, loc=<10.7246 4.8596 -6.7069>, rot=<0.0015 -0.8445 -0.5296 -0.0799>}
 dyn_pendant {parent=base_pendant, loc=<0 0 0>, rot=<0 0 0 1>}
 base_frontChains {parent=pelvis, loc=<10.3275 4.8744 -7.1511>, rot=<-0.3141 -0.8077 -0.2844 0.4099>}
 dyn_frontChains_01 {parent=base_frontChains, loc=<0 0 0>, rot=<0 0 0 1>}
 dyn_frontChains_02 {parent=dyn_frontChains_01, loc=<6.2287 0 0>, rot=<0 -0.2473 0 -0.9689>}
 dyn_frontChains_03 {parent=dyn_frontChains_02, loc=<7.0109 0 0>, rot=<0 0 0 1>}
 base_backChains {parent=pelvis, loc=<12.0378 -5.7447 4.5139>, rot=<0.2821 -0.6052 0.6325 0.3926>}
 dyn_backChains_01 {parent=base_backChains, loc=<0 0 0>, rot=<0 0 0 1>}
 dyn_backChains_02 {parent=dyn_backChains_01, loc=<3.654 0 0>, rot=<0 -0.2942 0 -0.9558>}
 dyn_backChains_03 {parent=dyn_backChains_02, loc=<5.8384 0 0>, rot=<0 0 0 1>}
]
variables [
 Crouching {type=bool, default=False}
 DBNO {type=bool, default=False}
]
