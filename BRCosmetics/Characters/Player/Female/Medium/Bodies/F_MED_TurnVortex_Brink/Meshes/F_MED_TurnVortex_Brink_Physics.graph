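# Physics asset for the F_MED_TurnVortex_Brink female character body, covering a necklace and the upper spine and neck around it.
# Single capsules sit on spine_05, neck_01, neck_02, base_necklace and dyn_necklace, with the larger torso and neck shapes apparently acting as colliders for the pendant.
# Constraints chain dyn_necklace through base_necklace and neck_01 to spine_05, with neck_02 branching off neck_01.
# Only the dyn_necklace joint is restricted, with limited swing and locked twist; the others rotate freely.
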PhysicsAsset /BRCosmetics/Characters/Player/Female/Medium/Bodies/F_MED_TurnVortex_Brink/Meshes/F_MED_TurnVortex_Brink_Physics
bodies {
 base_necklace [{
   shapeType=Capsule
   length=0.505
   radius=0.505
   shapeName=base_necklace_capsule
   center=<-0.1106 0.0014 -0.0061>
   rotation=<rot 0 -89.8186 -82.1408>
  }]
 dyn_necklace [{shapeType=Capsule, length=2.0546, radius=0.9398, shapeName=dyn_necklace_capsule, center=<2.8605 0 0.6708>, rotation=<rot 90 -90 0>}]
 neck_01 [{shapeType=Capsule, length=5, radius=4, shapeName=neck_01_capsule, center=<0.6622 1.6851 0>, rotation=<rot 90 -19.9082 0>}]
 neck_02 [{shapeType=Capsule, length=5, radius=4, shapeName=neck_02_capsule, center=<2.0222 0.7736 0>, rotation=<rot 90 -3.1862 0>}]
 spine_05 [{shapeType=Capsule, length=8.8258, radius=7.2439, center=<3.6418 -0.0132 0.0002>, rotation=<rot 0.0027 -90 0.0023>}]
}
constraints [
 {bone1Name=dyn_necklace, bone2Name=base_necklace, swing1LimitDegrees=20, swing2LimitDegrees=30, twistLimitDegrees=45, twistMotion=Locked}
 {
  bone1Name=neck_01
  bone2Name=spine_05
  swing1LimitDegrees=45
  swing1Motion=Free
  swing2LimitDegrees=45
  swing2Motion=Free
  twistLimitDegrees=45
  twistMotion=Free
 }
 {
  bone1Name=neck_02
  bone2Name=neck_01
  swing1LimitDegrees=45
  swing1Motion=Free
  swing2LimitDegrees=45
  swing2Motion=Free
  twistLimitDegrees=45
  twistMotion=Free
 }
 {
  bone1Name=base_necklace
  bone2Name=neck_01
  swing1LimitDegrees=45
  swing1Motion=Free
  swing2LimitDegrees=45
  swing2Motion=Free
  twistLimitDegrees=45
  twistMotion=Free
 }
]
