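# Post-process style Anim Blueprint for the TurnVortex Brink female body, adding secondary motion to its pendant, chains, laces and zippers.
# The attached parent's copied pose feeds a Control Rig, then pendant AnimDynamics variants blended by front-end idle and DBNO-carried state.
# A neck-relative RigidBody and chained AnimDynamics follow, then additive bone and look-at pendant corrections.
# The event graph updates gravity override and stance, zipline and front-end idle values.
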AnimBlueprint /BRCosmetics/Characters/Player/Female/Medium/Bodies/F_MED_TurnVortex_Brink/Meshes/F_MED_TurnVortex_Brink_AnimBP
target_skeleton=&/F_MED_TurnVortex_Brink_Skeleton
graphs [
 AnimGraph {nodes [
   n1 = "Transform (Modify) Bone - Bone: base_pendant"(
    ComponentPose=Local To Component(LocalPose=Use cached pose 'DynamicsCache'())
    Alpha=1
    Translation="0,0,0"
    Rotation="-2.277,-16.094,11.619"
    Scale="1,1,1"
   ) {
    rotationMode=BMM_Additive
    rotationSpace=BCS_BoneSpace
    scaleMode=BMM_Additive
    scaleSpace=BCS_BoneSpace
    translationMode=BMM_Additive
    translationSpace=BCS_BoneSpace
    alphaScaleBiasClamp {bInterpResult=true}
    boneToModify {boneName=base_pendant}
    rotation=<rot -2.277 -16.094 11.619>
   }
   n2 = Look At - Bone: dyn_pendant!(ComponentPose=@n1, bAlphaBoolEnabled:bool=True) -> (Pose) {
    alphaInputType=Bool
    alphaBoolBlend {blendInTime=0.2, blendOutTime=0.2}
    boneToModify {boneName=dyn_pendant}
    lookAtLocation {x=-11.0119}
    lookAtTarget {boneReference {boneName=calf_l}}
    lookAt_Axis {axis=<1 0>}
   }
   n3 = "Transform (Modify) Bone - Bone: base_pendant"!(ComponentPose=@n2, bAlphaBoolEnabled:bool=True, Translation="0,0,0", Rotation="-10,0,0", Scale="1,1,1") -> (Pose) {
    alphaInputType=Bool
    rotationMode=BMM_Additive
    rotationSpace=BCS_BoneSpace
    scaleMode=BMM_Additive
    scaleSpace=BCS_BoneSpace
    translationMode=BMM_Additive
    translationSpace=BCS_BoneSpace
    alphaBoolBlend {blendInTime=0.2, blendOutTime=0.2}
    alphaScaleBiasClamp {bInterpResult=true}
    boneToModify {boneName=base_pendant}
    rotation {pitch=-10}
   }
   n4 = "Transform (Modify) Bone - Bone: base_pendant"(ComponentPose=@n3, bAlphaBoolEnabled=True, Translation="0,0,0", Rotation="40,0.014,0.014", Scale="1,1,1") {
    alphaInputType=Bool
    rotationMode=BMM_Additive
    rotationSpace=BCS_BoneSpace
    scaleMode=BMM_Additive
    scaleSpace=BCS_BoneSpace
    translationMode=BMM_Additive
    translationSpace=BCS_BoneSpace
    alphaBoolBlend {blendInTime=0.2, blendOutTime=0.2}
    alphaScaleBiasClamp {bInterpResult=true}
    boneToModify {boneName=base_pendant}
    rotation=<rot 40 0.014 0.014>
   }
   Output Pose(Result=Component To Local(ComponentPose=@n4)) {layerGroup=None}
   n5 = Control Rig!(Source=Copy Pose From Mesh() {bUseAttachedParent=true}, Alpha, Crouching:bool=Get bIsCrouching(), DBNO:bool=Get bIsDBNO()) -> (Pose) {controlRigAssetReference {blueprintRigClass=&/F_MED_TurnVortex_Brink_CtrlRig.F_MED_TurnVortex_Brink_CtrlRig_C}}
   Save cached pose 'ControlRigCache'(Pose=@n5) {cacheName=ControlRigCache}
   n6 = Anim Dynamics - Bone: dyn_pendant!(
    ComponentPose=Local To Component(LocalPose=Use cached pose 'ControlRigCache'())
    Alpha=1
    GravityOverride:Vector="0,0,-980"
    ExternalForce="0,0,0"
   ) -> (Pose) {
    angularDampingOverride=0.999
    angularSpringConstant=5000
    bAngularSpring=true
    bOverrideAngularDamping=true
    bUseGravityOverride=true
    bUsePlanarLimit=false
    bUseSphericalLimits=true
    gravityScale=0.35
    simulationSpace=RootRelative
    alphaScaleBiasClamp {bInterpResult=true}
    boundBone {boneName=dyn_pendant}
    componentAppliedLinearAccClamp=<2000 2000 2000>
    componentLinearAccScale=<1 1 1>
    componentLinearVelScale=<1 1 1>
    gravityOverride {z=-980}
    physicsBodyDefinitions [{
      boundBone {boneName=dyn_pendant}
      boxExtents=<20 30 30>
      constraintSetup {angularLimitsMax=<10 0 60>, angularLimitsMin=<-10 -150 -100>, angularTarget=<1 0 0>}
      localJointOffset=<5 0 0>
     }]
    simSpaceSettings {maxAngularAcceleration=1000, maxAngularVelocity=1000, simSpaceAngularAlpha=0.7, externalAngularVelocity=<1 1 1>}
    sphericalLimits [{limitRadius=6, drivingBone {boneName=pelvis}, sphereLocalOffset=<0 0.5 -6>}]
   }
   n7 = Anim Dynamics - Bone: dyn_pendant(ComponentPose=Local To Component(LocalPose=Use cached pose 'ControlRigCache'()), Alpha=1, GravityOverride="0,0,-980", ExternalForce="0,0,0") {
    angularDampingOverride=0.999
    angularSpringConstant=5000
    bAngularSpring=true
    bOverrideAngularDamping=true
    bUseGravityOverride=true
    bUsePlanarLimit=false
    bUseSphericalLimits=true
    gravityScale=0.35
    simulationSpace=RootRelative
    alphaScaleBiasClamp {bInterpResult=true}
    boundBone {boneName=dyn_pendant}
    componentAppliedLinearAccClamp=<2000 2000 2000>
    componentLinearAccScale=<1 1 1>
    componentLinearVelScale=<1 1 1>
    gravityOverride {z=-980}
    physicsBodyDefinitions [{
      boundBone {boneName=dyn_pendant}
      boxExtents=<20 30 30>
      constraintSetup {angularLimitsMax=<10 0 60>, angularLimitsMin=<-10 -150 -100>, angularTarget=<1 0 0>}
      localJointOffset=<5 0 0>
     }]
    simSpaceSettings {maxAngularAcceleration=1000, maxAngularVelocity=1000, simSpaceAngularAlpha=0.7, externalAngularVelocity=<1 1 1>}
    sphericalLimits [{limitRadius=7.5, drivingBone {boneName=pelvis}, sphereLocalOffset=<0 0.5 -6>}]
   }
   n8 = Blend Poses by bool(BlendPose_0=Component To Local(ComponentPose=@n6), BlendPose_1=Component To Local(ComponentPose=@n7), bActiveValue=Get bIsDBNOCarried())
   n9 = Anim Dynamics - Bone: dyn_pendant(ComponentPose=Local To Component(LocalPose=Use cached pose 'ControlRigCache'()), Alpha=1, GravityOverride="0,0,-980", ExternalForce="0,0,0") {
    angularDampingOverride=0.999
    angularSpringConstant=5000
    bAngularSpring=true
    bOverrideAngularDamping=true
    bUseGravityOverride=true
    bUsePlanarLimit=false
    bUseSphericalLimits=true
    gravityScale=0.35
    simulationSpace=RootRelative
    alphaScaleBiasClamp {bInterpResult=true}
    boundBone {boneName=dyn_pendant}
    componentAppliedLinearAccClamp=<2000 2000 2000>
    componentLinearAccScale=<1 1 1>
    componentLinearVelScale=<1 1 1>
    gravityOverride {z=-980}
    physicsBodyDefinitions [{
      boundBone {boneName=dyn_pendant}
      boxExtents=<20 30 30>
      constraintSetup {angularLimitsMax=<10 0 60>, angularLimitsMin=<-10 -150 -100>, angularTarget=<1 0 0>}
      localJointOffset=<5 0 0>
     }]
    simSpaceSettings {maxAngularAcceleration=1000, maxAngularVelocity=1000, simSpaceAngularAlpha=0.3, externalAngularVelocity=<1 1 1>}
    sphericalLimits [{limitRadius=7.5, drivingBone {boneName=pelvis}, sphereLocalOffset=<0 0.5 -6>}]
   }
   n10 = Component To Local(ComponentPose=@n9)
   n11 = Get FE IdleBool() -> (FE IdleBool:bool)
   n12 = RigidBody!(
    ComponentPose=Local To Component(LocalPose=Blend Poses by bool(BlendPose_0=@n10, BlendPose_1=@n8, bActiveValue=@n11))
    Alpha
    OverrideWorldGravity:Vector="0,0,-980"
    ExternalForce
   ) -> (Pose) {
    bOverrideWorldGravity=true
    overridePhysicsAsset=&/F_MED_TurnVortex_Brink_Physics
    simulationSpace=BaseBoneSpace
    baseBoneRef {boneName=neck_01}
    componentAppliedLinearAccClamp=<3000 3000 3000>
    simSpaceSettings {
     maxAngularAcceleration=150
     maxAngularVelocity=150
     maxLinearAcceleration=300
     maxLinearVelocity=600
     worldAlpha=1
     externalLinearDragV=<1 1 1>
    }
   }
   n13 = Anim Dynamics - Bone: dyn_frontChains_01!(ComponentPose=@n12, Alpha=1, GravityOverride:Vector="0,0,-980", ExternalForce="0,0,0") -> (Pose) {
    angularDampingOverride=1
    angularSpringConstant=10000
    bAngularSpring=true
    bOverrideAngularDamping=true
    bUseGravityOverride=true
    gravityScale=0.35
    simulationSpace=RootRelative
    boundBone {boneName=dyn_frontChains_01}
    componentAppliedLinearAccClamp=<2000 2000 2000>
    componentLinearAccScale=<0.5 0.5 0.5>
    componentLinearVelScale=<0.5 0.5 0.5>
    gravityOverride {z=-980}
    physicsBodyDefinitions [{
      boundBone {boneName=dyn_frontChains_01}
      boxExtents=<20 30 30>
      constraintSetup {angularLimitsMax=<3 0 3>, angularLimitsMin=<-3 0 0>, angularTarget=<1 0 0>}
      localJointOffset=<2 0 0>
     }]
    simSpaceSettings {maxAngularAcceleration=1000, maxAngularVelocity=1000, simSpaceAngularAlpha=0.5}
   }
   n14 = Anim Dynamics - Bone: dyn_frontChains_02!(ComponentPose=@n13, Alpha=1, GravityOverride:Vector="0,0,-980", ExternalForce="0,0,0") -> (Pose) {
    angularDampingOverride=1
    angularSpringConstant=10000
    bAngularSpring=true
    bOverrideAngularDamping=true
    bUseGravityOverride=true
    gravityScale=0.35
    simulationSpace=RootRelative
    boundBone {boneName=dyn_frontChains_02}
    componentAppliedLinearAccClamp=<2000 2000 2000>
    componentLinearAccScale=<0.5 0.5 0.5>
    componentLinearVelScale=<0.5 0.5 0.5>
    gravityOverride {z=-980}
    physicsBodyDefinitions [{
      boundBone {boneName=dyn_frontChains_02}
      boxExtents=<20 30 30>
      constraintSetup {angularLimitsMax=<3 0 6>, angularLimitsMin=<-3 0 0>, angularTarget=<1 0 0>}
      localJointOffset=<2 0 0>
     }]
    simSpaceSettings {maxAngularAcceleration=1000, maxAngularVelocity=1000, simSpaceAngularAlpha=0.5}
   }
   n15 = Anim Dynamics - Bone: dyn_frontChains_03!(ComponentPose=@n14, Alpha=1, GravityOverride:Vector="0,0,-980", ExternalForce="0,0,0") -> (Pose) {
    angularDampingOverride=1
    angularSpringConstant=10000
    bAngularSpring=true
    bOverrideAngularDamping=true
    bUseGravityOverride=true
    gravityScale=0.35
    simulationSpace=RootRelative
    boundBone {boneName=dyn_frontChains_03}
    componentAppliedLinearAccClamp=<2000 2000 2000>
    componentLinearAccScale=<0.5 0.5 0.5>
    componentLinearVelScale=<0.5 0.5 0.5>
    gravityOverride {z=-980}
    physicsBodyDefinitions [{
      boundBone {boneName=dyn_frontChains_03}
      boxExtents=<20 30 30>
      constraintSetup {angularLimitsMax=<3 0 12>, angularLimitsMin=<-3 0 0>, angularTarget=<1 0 0>}
      localJointOffset=<2 0 0>
     }]
    simSpaceSettings {maxAngularAcceleration=1000, maxAngularVelocity=1000, simSpaceAngularAlpha=0.5}
   }
   n16 = Anim Dynamics - Bone: dyn_backChains_01!(ComponentPose=@n15, Alpha=1, GravityOverride:Vector="0,0,-980", ExternalForce="0,0,0") -> (Pose) {
    angularDampingOverride=1
    angularSpringConstant=10000
    bAngularSpring=true
    bOverrideAngularDamping=true
    bUseGravityOverride=true
    gravityScale=0.35
    simulationSpace=RootRelative
    boundBone {boneName=dyn_backChains_01}
    componentAppliedLinearAccClamp=<2000 2000 2000>
    componentLinearAccScale=<0.5 0.5 0.5>
    componentLinearVelScale=<0.5 0.5 0.5>
    gravityOverride {z=-980}
    physicsBodyDefinitions [{
      boundBone {boneName=dyn_backChains_01}
      boxExtents=<20 30 30>
      constraintSetup {angularLimitsMax=<3 0 0>, angularLimitsMin=<-3 0 -3>, angularTarget=<1 0 0>}
      localJointOffset=<2 0 0>
     }]
    simSpaceSettings {maxAngularAcceleration=1000, maxAngularVelocity=1000, simSpaceAngularAlpha=1}
   }
   n17 = Anim Dynamics - Bone: dyn_backChains_02!(ComponentPose=@n16, Alpha=1, GravityOverride:Vector="0,0,-980", ExternalForce="0,0,0") -> (Pose) {
    angularDampingOverride=1
    angularSpringConstant=10000
    bAngularSpring=true
    bOverrideAngularDamping=true
    bUseGravityOverride=true
    gravityScale=0.35
    simulationSpace=RootRelative
    boundBone {boneName=dyn_backChains_02}
    componentAppliedLinearAccClamp=<2000 2000 2000>
    componentLinearAccScale=<0.5 0.5 0.5>
    componentLinearVelScale=<0.5 0.5 0.5>
    gravityOverride {z=-980}
    physicsBodyDefinitions [{
      boundBone {boneName=dyn_backChains_02}
      boxExtents=<20 30 30>
      constraintSetup {angularLimitsMax=<3 0 0>, angularLimitsMin=<-3 0 -12>, angularTarget=<1 0 0>}
      localJointOffset=<2 0 0>
     }]
    simSpaceSettings {maxAngularAcceleration=1000, maxAngularVelocity=1000, simSpaceAngularAlpha=1}
   }
   n18 = Anim Dynamics - Bone: dyn_backChains_03!(ComponentPose=@n17, Alpha=1, GravityOverride:Vector="0,0,-980", ExternalForce="0,0,0") -> (Pose) {
    angularDampingOverride=1
    angularSpringConstant=10000
    bAngularSpring=true
    bOverrideAngularDamping=true
    bUseGravityOverride=true
    gravityScale=0.35
    simulationSpace=RootRelative
    boundBone {boneName=dyn_backChains_03}
    componentAppliedLinearAccClamp=<2000 2000 2000>
    componentLinearAccScale=<0.5 0.5 0.5>
    componentLinearVelScale=<0.5 0.5 0.5>
    gravityOverride {z=-980}
    physicsBodyDefinitions [{
      boundBone {boneName=dyn_backChains_03}
      boxExtents=<20 30 30>
      constraintSetup {angularLimitsMax=<3 0 0>, angularLimitsMin=<-3 0 -12>, angularTarget=<1 0 0>}
      localJointOffset=<2 0 0>
     }]
    simSpaceSettings {maxAngularAcceleration=1000, maxAngularVelocity=1000, simSpaceAngularAlpha=1}
   }
   n19 = Anim Dynamics - Bone: dyn_pendant(ComponentPose=@n18, Alpha=1, GravityOverride="0,0,-980", ExternalForce="0,0,0") {
    angularDampingOverride=1
    angularSpringConstant=2000
    bAngularSpring=true
    bOverrideAngularDamping=true
    bUseGravityOverride=true
    bUsePlanarLimit=false
    bUseSphericalLimits=true
    gravityScale=0.35
    simulationSpace=RootRelative
    alphaScaleBiasClamp {bInterpResult=true}
    boundBone {boneName=dyn_pendant}
    componentAppliedLinearAccClamp=<3000 3000 3000>
    componentLinearAccScale=<1 1 1>
    componentLinearVelScale=<1 1 1>
    gravityOverride {z=-980}
    physicsBodyDefinitions [{
      boundBone {boneName=dyn_pendant}
      boxExtents=<20 30 30>
      constraintSetup {angularLimitsMax=<7 0 20>, angularLimitsMin=<-7 -50 -20>, angularTarget=<1 0 0>}
      localJointOffset=<4 0 0>
     }]
    simSpaceSettings {maxAngularAcceleration=1000, maxAngularVelocity=1000, simSpaceAngularAlpha=1, externalAngularVelocity=<1 1 1>}
    sphericalLimits [{limitRadius=10, drivingBone {boneName=thigh_l}, sphereLocalOffset=<-2 0.5 0>}]
   }
   n20 = Anim Dynamics - Bone: dyn_pendant!(ComponentPose=@n19, bAlphaBoolEnabled:bool=True, GravityOverride="0,0,-980", ExternalForce="0,0,0") -> (Pose) {
    alphaInputType=Bool
    angularDampingOverride=0.999
    angularSpringConstant=5000
    bAngularSpring=true
    bOverrideAngularDamping=true
    bUseGravityOverride=true
    bUseSphericalLimits=true
    gravityScale=0.35
    simulationSpace=RootRelative
    alphaBoolBlend {blendInTime=0.2, blendOutTime=0.2}
    alphaScaleBiasClamp {bInterpResult=true}
    boundBone {boneName=dyn_pendant}
    componentAppliedLinearAccClamp=<500 500 500>
    componentLinearAccScale=<0.5 0.5 0.5>
    componentLinearVelScale=<1 1 1>
    gravityOverride {z=-980}
    physicsBodyDefinitions [{
      boundBone {boneName=dyn_pendant}
      boxExtents=<20 30 30>
      constraintSetup {angularLimitsMax=<10 0 60>, angularLimitsMin=<-10 -150 -100>, angularTarget=<1 0 0>}
      localJointOffset=<5 0 0>
     }]
    simSpaceSettings {maxAngularAcceleration=1000, maxAngularVelocity=1000, externalAngularVelocity=<1 1 1>}
    sphericalLimits [{limitRadius=40, drivingBone {boneName=pelvis}, sphereLocalOffset=<0 -33 0>}]
   }
   n21 = Anim Dynamics - Bone: dyn_race_02_r!(ComponentPose=@n20, Alpha=1, GravityOverride:Vector="0,0,-980", ExternalForce="0,0,0") -> (Pose) {
    angularDampingOverride=0.999
    angularSpringConstant=10000
    bAngularSpring=true
    bOverrideAngularDamping=true
    bUseGravityOverride=true
    gravityScale=0.35
    simulationSpace=RootRelative
    boundBone {boneName=dyn_race_02_r}
    componentAppliedLinearAccClamp=<3000 3000 3000>
    componentLinearAccScale=<1 1 1>
    componentLinearVelScale=<1 1 1>
    gravityOverride {z=-980}
    physicsBodyDefinitions [{
      boundBone {boneName=dyn_race_02_r}
      boxExtents=<30 30 30>
      constraintSetup {angularLimitsMax=<15 50 40>, angularLimitsMin=<-15 -15 0>, angularTarget=<1 0 0>}
      localJointOffset=<4 0 0>
     }]
    simSpaceSettings {maxAngularAcceleration=100, maxAngularVelocity=100, simSpaceAngularAlpha=0.7, externalAngularVelocity=<1 1 1>}
   }
   n22 = Anim Dynamics - Bone: dyn_race_01_r!(ComponentPose=@n21, Alpha=1, GravityOverride:Vector="0,0,-980", ExternalForce="0,0,0") -> (Pose) {
    angularDampingOverride=0.999
    angularSpringConstant=10000
    bAngularSpring=true
    bOverrideAngularDamping=true
    bUseGravityOverride=true
    gravityScale=0.35
    simulationSpace=RootRelative
    boundBone {boneName=dyn_race_01_r}
    componentAppliedLinearAccClamp=<3000 3000 3000>
    componentLinearAccScale=<1 1 1>
    componentLinearVelScale=<1 1 1>
    gravityOverride {z=-980}
    physicsBodyDefinitions [{
      boundBone {boneName=dyn_race_01_r}
      boxExtents=<20 30 30>
      constraintSetup {angularLimitsMax=<15 50 40>, angularLimitsMin=<-15 -15 0>, angularTarget=<1 0 0>}
      localJointOffset=<4 1 0>
     }]
    simSpaceSettings {maxAngularAcceleration=100, maxAngularVelocity=100, simSpaceAngularAlpha=0.7, externalAngularVelocity=<1 1 1>}
   }
   n23 = Anim Dynamics - Bone: dyn_rounfLaces_01_r!(ComponentPose=@n22, Alpha=1, GravityOverride:Vector="0,0,-980", ExternalForce="0,0,0") -> (Pose) {
    angularDampingOverride=1
    angularSpringConstant=10000
    bAngularSpring=true
    bOverrideAngularDamping=true
    bUseGravityOverride=true
    gravityScale=0.35
    simulationSpace=RootRelative
    boundBone {boneName=dyn_rounfLaces_01_r}
    componentAppliedLinearAccClamp=<3000 3000 3000>
    componentLinearAccScale=<1 1 1>
    componentLinearVelScale=<1 1 1>
    gravityOverride {z=-980}
    physicsBodyDefinitions [{
      boundBone {boneName=dyn_rounfLaces_01_r}
      boxExtents=<20 30 30>
      constraintSetup {angularLimitsMax=<15 50 0>, angularLimitsMin=<-15 -15 -40>, angularTarget=<1 0 0>}
      localJointOffset=<2 0 0>
     }]
    simSpaceSettings {maxAngularAcceleration=100, maxAngularVelocity=100, simSpaceAngularAlpha=0.7, externalAngularVelocity=<1 1 1>}
   }
   n24 = Anim Dynamics - Bone: dyn_rounfLaces_02_r!(ComponentPose=@n23, Alpha=1, GravityOverride:Vector="0,0,-980", ExternalForce="0,0,0") -> (Pose) {
    angularDampingOverride=1
    angularSpringConstant=10000
    bAngularSpring=true
    bOverrideAngularDamping=true
    bUseGravityOverride=true
    gravityScale=0.35
    simulationSpace=RootRelative
    boundBone {boneName=dyn_rounfLaces_02_r}
    componentAppliedLinearAccClamp=<3000 3000 3000>
    componentLinearAccScale=<1 1 1>
    componentLinearVelScale=<1 1 1>
    gravityOverride {z=-980}
    physicsBodyDefinitions [{
      boundBone {boneName=dyn_rounfLaces_02_r}
      boxExtents=<20 30 30>
      constraintSetup {angularLimitsMax=<15 50 0>, angularLimitsMin=<-15 -15 -40>, angularTarget=<1 0 0>}
      localJointOffset=<2 0 0>
     }]
    simSpaceSettings {maxAngularAcceleration=100, maxAngularVelocity=100, simSpaceAngularAlpha=0.7, externalAngularVelocity=<1 1 1>}
   }
   n25 = Anim Dynamics - Bone: dyn_lace_02_l!(ComponentPose=@n24, Alpha=1, GravityOverride:Vector="0,0,-980", ExternalForce="0,0,0") -> (Pose) {
    angularDampingOverride=0.999
    angularSpringConstant=10000
    bAngularSpring=true
    bOverrideAngularDamping=true
    bUseGravityOverride=true
    gravityScale=0.35
    simulationSpace=RootRelative
    boundBone {boneName=dyn_lace_02_l}
    componentAppliedLinearAccClamp=<3000 3000 3000>
    componentLinearAccScale=<1 1 1>
    componentLinearVelScale=<1 1 1>
    gravityOverride {z=-980}
    physicsBodyDefinitions [{
      boundBone {boneName=dyn_lace_02_l}
      boxExtents=<30 30 30>
      constraintSetup {angularLimitsMax=<15 50 40>, angularLimitsMin=<-15 -15 0>, angularTarget=<1 0 0>}
      localJointOffset=<4 0 0>
     }]
    simSpaceSettings {maxAngularAcceleration=1000, maxAngularVelocity=1000, simSpaceAngularAlpha=0.7, externalAngularVelocity=<1 1 1>}
   }
   n26 = Anim Dynamics - Bone: dyn_lace_01_l!(ComponentPose=@n25, Alpha=1, GravityOverride:Vector="0,0,-980", ExternalForce="0,0,0") -> (Pose) {
    angularDampingOverride=0.999
    angularSpringConstant=10000
    bAngularSpring=true
    bOverrideAngularDamping=true
    bUseGravityOverride=true
    gravityScale=0.35
    simulationSpace=RootRelative
    boundBone {boneName=dyn_lace_01_l}
    componentAppliedLinearAccClamp=<3000 3000 3000>
    componentLinearAccScale=<1 1 1>
    componentLinearVelScale=<1 1 1>
    gravityOverride {z=-980}
    physicsBodyDefinitions [{
      boundBone {boneName=dyn_lace_01_l}
      boxExtents=<20 30 30>
      constraintSetup {angularLimitsMax=<15 15 40>, angularLimitsMin=<-15 -50 0>, angularTarget=<1 0 0>}
      localJointOffset=<4 1 0>
     }]
    simSpaceSettings {maxAngularAcceleration=1000, maxAngularVelocity=1000, simSpaceAngularAlpha=0.7, externalAngularVelocity=<1 1 1>}
   }
   n27 = Anim Dynamics - Bone: dyn_rounfLaces_01_l!(ComponentPose=@n26, Alpha=1, GravityOverride:Vector="0,0,-980", ExternalForce="0,0,0") -> (Pose) {
    angularDampingOverride=1
    angularSpringConstant=10000
    bAngularSpring=true
    bOverrideAngularDamping=true
    bUseGravityOverride=true
    gravityScale=0.35
    simulationSpace=RootRelative
    boundBone {boneName=dyn_rounfLaces_01_l}
    componentAppliedLinearAccClamp=<3000 3000 3000>
    componentLinearAccScale=<1 1 1>
    componentLinearVelScale=<1 1 1>
    gravityOverride {z=-980}
    physicsBodyDefinitions [{
      boundBone {boneName=dyn_rounfLaces_01_l}
      boxExtents=<20 30 30>
      constraintSetup {angularLimitsMax=<15 15 40>, angularLimitsMin=<-15 -50 0>, angularTarget=<1 0 0>}
      localJointOffset=<2 0 0>
     }]
    simSpaceSettings {maxAngularAcceleration=1000, maxAngularVelocity=1000, simSpaceAngularAlpha=0.7, externalAngularVelocity=<1 1 1>}
   }
   n28 = Anim Dynamics - Bone: dyn_rounfLaces_02_l!(ComponentPose=@n27, Alpha=1, GravityOverride:Vector="0,0,-980", ExternalForce="0,0,0") -> (Pose) {
    angularDampingOverride=1
    angularSpringConstant=10000
    bAngularSpring=true
    bOverrideAngularDamping=true
    bUseGravityOverride=true
    gravityScale=0.35
    simulationSpace=RootRelative
    boundBone {boneName=dyn_rounfLaces_02_l}
    componentAppliedLinearAccClamp=<3000 3000 3000>
    componentLinearAccScale=<1 1 1>
    componentLinearVelScale=<1 1 1>
    gravityOverride {z=-980}
    physicsBodyDefinitions [{
      boundBone {boneName=dyn_rounfLaces_02_l}
      boxExtents=<20 30 30>
      constraintSetup {angularLimitsMax=<15 15 40>, angularLimitsMin=<-15 -50 0>, angularTarget=<1 0 0>}
      localJointOffset=<2 0 0>
     }]
    simSpaceSettings {maxAngularAcceleration=1000, maxAngularVelocity=1000, simSpaceAngularAlpha=0.7, externalAngularVelocity=<1 1 1>}
   }
   n29 = Anim Dynamics - Bone: dyn_zipper_l!(ComponentPose=@n28, Alpha=1, GravityOverride:Vector="0,0,-980", ExternalForce="0,0,0") -> (Pose) {
    angularDampingOverride=0.9999
    angularSpringConstant=8000
    bAngularSpring=true
    bOverrideAngularDamping=true
    bUseGravityOverride=true
    bUsePlanarLimit=false
    gravityScale=0.35
    simulationSpace=RootRelative
    alphaBoolBlend {blendOption=Linear}
    boundBone {boneName=dyn_zipper_l}
    componentAppliedLinearAccClamp=<1400 1400 1400>
    componentLinearAccScale=<1 1 1>
    componentLinearVelScale=<1 1 1>
    gravityOverride {z=-980}
    physicsBodyDefinitions [{
      boundBone {boneName=dyn_zipper_l}
      boxExtents=<20 15 10>
      constraintSetup {angularLimitsMax=<15 45 90>, angularLimitsMin=<-15 0 0>, angularTarget=<1 0 0>}
      localJointOffset=<4 0 1>
     }]
    simSpaceSettings {simSpaceAngularAlpha=0.35}
   }
   n30 = Anim Dynamics - Bone: dyn_zipper_r!(ComponentPose=@n29, Alpha=1, GravityOverride:Vector="0,0,-980", ExternalForce="0,0,0") -> (Pose) {
    angularDampingOverride=0.9999
    angularSpringConstant=8000
    bAngularSpring=true
    bOverrideAngularDamping=true
    bUseGravityOverride=true
    bUsePlanarLimit=false
    gravityScale=0.35
    simulationSpace=RootRelative
    alphaBoolBlend {blendOption=Linear}
    boundBone {boneName=dyn_zipper_r}
    componentAppliedLinearAccClamp=<1400 1400 1400>
    componentLinearAccScale=<1 1 1>
    componentLinearVelScale=<1 1 1>
    gravityOverride {z=-980}
    physicsBodyDefinitions [{
      boundBone {boneName=dyn_zipper_r}
      boxExtents=<20 15 10>
      constraintSetup {angularLimitsMax=<15 45 90>, angularLimitsMin=<-15 0 0>, angularTarget=<1 0 0>}
      localJointOffset=<4 0 1>
     }]
    simSpaceSettings {simSpaceAngularAlpha=0.35}
   }
   Save cached pose 'DynamicsCache'(Pose=Component To Local(ComponentPose=@n30)) {cacheName=DynamicsCache}
  ]}
 EventGraph {nodes [
   Event BlueprintUpdateAnimation()
   > n1 = Sequence()
   > n2 = "Gravity Override State Function (Nativized)"!(
    execute
    CharacterPartAnimInstance:Custom Character Part Anim Instance Object Reference=Self-Reference()
    GravityOverrideParameterSet:Fort Gravity Override Parameter Set Structure=Get Gravity Override Params() -> (Gravity Override Params:Fort Gravity Override Parameter Set Structure)
    RigidBodyAnimNodeParameterSet:Fort Rigid Body Anim Node Parameter Set Structure=Get RBAN Params() -> (RBAN Params:Fort Rigid Body Anim Node Parameter Set Structure)
    DebugDraw:bool=false
    GameTimeInSeconds:float=GetGameTimeInSeconds()
    DebugJointName:Name=root
    StateLogicLayerName:Name=none
   ) -> (
    then
    OutGravityOverride:Vector
    OutCurrentState:Name=None
    OutGravityOverrideParameters:Fort Gravity Override Parameters Structure
    OutRigidBodyAnimNodeParameters:Fort Rigid Body Anim Node Parameters Structure
   )
   > Set Gravity Override(Gravity Override=@n2)
   n3 = NOT Boolean(A=Get bIsSliding())
   n4 = Get bIsCrouching()
   Set Alpha_Crouching(execute=@n1.then_1, Alpha_Crouching:double="To Float (Boolean)"(InBool=AND Boolean(A=@n4, B=@n3))) -> (Output_Get:double)
   n5 = NOT Boolean(A=Get bIsMoving2D())
   n6 = Get bIsCrouching()
   > Set Alpha_CrouchingIdle(Alpha_CrouchingIdle:double="To Float (Boolean)"(InBool=AND Boolean(A=@n5, B=@n6))) -> (Output_Get:double)
   n7 = NOT Boolean(A=Get OnZipline_Float() -> (OnZipline_Float:bool))
   n8 = OR Boolean(A=NOT Boolean(A=Get bIsCrouching()), B=Get bIsSliding())
   > Set Alpha_Standing(Alpha_Standing:double="To Float (Boolean)"(InBool=AND Boolean(A=@n8, B=@n7))) -> (Output_Get:double)
   n9 = Get bIsMoving2D()
   n10 = Get bIsSkyDiving()
   n11 = Get bIsPlayingEmote()
   > n12 = Branch(Condition=AND Boolean!(A=Get bIsInFrontEnd(), B=NOT Boolean(A=@n9), C:bool=NOT Boolean(A=@n10), D:bool=NOT Boolean(A=@n11)) -> (ReturnValue))
   > Set FE Idle(FE Idle:double=0.35) -> (Output_Get:double)
   > Set FE IdleBool(FE IdleBool:bool=true) -> (Output_Get:bool)
   Set FE Idle(execute=@n12.else, FE Idle=1)
   > Set FE IdleBool(FE IdleBool=false)
   n13 = Event BlueprintProcessAnimInputEvent()
   > n14 = Branch(Condition=IsNamed(self=@n13.AnimInputEvent, InEventName=aie.Player.Zipline.Entered))
   > Set OnZipline_Float(OnZipline_Float:bool=1) -> (Output_Get:bool)
   Branch(execute=@n14.else, Condition=IsNamed(self=@n13.AnimInputEvent, InEventName=aie.Player.Zipline.Exited))
   > Set OnZipline_Float(OnZipline_Float=false)
  ]}
]
variables [
 Alpha_Crouching {type=number}
 Alpha_CrouchingIdle {type=number}
 Alpha_Standing {type=number}
 FE Idle {type=number}
 FE IdleBool {type=boolean}
 Gravity Override {type=Vector}
 Gravity Override Params {
  type=FortGravityOverrideParameterSet
  default=(Skydive_GoingUp=(GravityOverride=()),Skydive_FloatingDown=(GravityOverride=()),Skydive_DivingDown=(GravityOverride=()),Skydive_Parachuting=(GravityOverride=()),OnGround_Standing=(GravityOverride=()),OnGround_Moving=(GravityOverride=()),OnGround_Moving_HighRotationRate=(GravityOverride=()),OnGround_Crouch_Moving=(GravityOverride=()),EmoteOrMelee=(GravityOverride=()),Swimming=(GravityOverride=()),Front_End=(GravityOverride=()),DBNO=(GravityOverride=()),DBNO_Carried=(GravityOverride=()),OnGround_Falling=(GravityOverride=()),Ziplining=(GravityOverride=()),Skydive_Surfing=(GravityOverride=()),Riding_Standing=(GravityOverride=()),Riding_Moving=(GravityOverride=()),Riding_Moving_Targeting=(GravityOverride=()),Tactical_Sprint=(GravityOverride=()),Rail_Grinding=(GravityOverride=()),)
 }
 OnZipline_Float {type=boolean}
 RBAN Params {
  type=FortRigidBodyAnimNodeParameterSet
  default=(Skydive_GoingUp=(RigidBodyAnimNode=()),Skydive_FloatingDown=(RigidBodyAnimNode=()),Skydive_DivingDown=(RigidBodyAnimNode=()),Skydive_Parachuting=(RigidBodyAnimNode=()),OnGround_Standing=(RigidBodyAnimNode=()),OnGround_Moving=(RigidBodyAnimNode=()),OnGround_Moving_HighRotationRate=(RigidBodyAnimNode=()),OnGround_Crouch_Moving=(RigidBodyAnimNode=()),EmoteOrMelee=(RigidBodyAnimNode=()),Swimming=(RigidBodyAnimNode=()),Front_End=(RigidBodyAnimNode=()),DBNO=(RigidBodyAnimNode=()),DBNO_Carried=(RigidBodyAnimNode=()),OnGround_Falling=(RigidBodyAnimNode=()),Ziplining=(RigidBodyAnimNode=()),Skydive_Surfing=(RigidBodyAnimNode=()),Riding_Standing=(RigidBodyAnimNode=()),Riding_Moving=(RigidBodyAnimNode=()),Riding_Moving_Targeting=(RigidBodyAnimNode=()),Tactical_Sprint=(RigidBodyAnimNode=()),Rail_Grinding=(RigidBodyAnimNode=()),)
 }
]
